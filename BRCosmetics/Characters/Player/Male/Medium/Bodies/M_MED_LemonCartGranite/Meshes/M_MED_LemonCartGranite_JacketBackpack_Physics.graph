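# Physics asset for the jacket/backpack mesh of the male medium LemonCartGranite player cosmetic body, simulating jacket motion.
# Seven capsule chains (front, side and rear left/right, plus rear center) start at tiny MSTR bodies attached to spine_04 with free rotation, using tight angular limits down each chain.
# Free-rotation cross-constraints tie rear chains together and front to sides.
# Pelvis, spine and thigh capsules provide body volumes.
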PhysicsAsset /BRCosmetics/Characters/Player/Male/Medium/Bodies/M_MED_LemonCartGranite/Meshes/M_MED_LemonCartGranite_JacketBackpack_Physics
bodies {
 dyn_jacket_front_l_01 [{shapeType=Capsule, length=8.5, radius=5, shapeName=dyn_jacket_front_l_01_capsule, center=<3 0 6>, rotation=<rot 0.0002 0.0001 79.6185>}]
 dyn_jacket_front_l_02 [{shapeType=Capsule, length=9.75, radius=5.75, shapeName=dyn_jacket_front_l_02_capsule, center=<3 0 6>, rotation=<rot 0.0001 0 79.8555>}]
 dyn_jacket_front_l_03 [{
   shapeType=Capsule
   length=13.875
   radius=5.3125
   shapeName=dyn_jacket_front_l_03_capsule
   center=<3 0 6>
   rotation=<rot 0.0003 0.0001 82.7202>
  }]
 dyn_jacket_front_l_04 [{shapeType=Capsule, length=11.0625, radius=7.3125, shapeName=dyn_jacket_front_l_04_capsule, center=<3 0 6>, rotation=<rot 0.0001 0 77.7655>}]
 dyn_jacket_front_l_05 [{
   shapeType=Capsule
   length=14.5625
   radius=7.6875
   shapeName=dyn_jacket_front_l_05_capsule
   center=<6.4606 -0.5341 7.5755>
   rotation=<rot 0.0003 0.0001 71.2733>
  }]
 dyn_jacket_front_l_MSTR [{shapeType=Capsule, length=0.505, radius=0.505, shapeName=dyn_jacket_front_l_MSTR_capsule, rotation=<rot 0 -90 0>}]
 dyn_jacket_front_r_01 [{shapeType=Capsule, length=8.5, radius=5, shapeName=dyn_jacket_front_l_01_capsule, center=<-3 0 -6>, rotation=<rot 0.0002 0.0001 -100.3815>}]
 dyn_jacket_front_r_02 [{shapeType=Capsule, length=9.75, radius=5.75, shapeName=dyn_jacket_front_l_02_capsule, center=<-3 0 -6>, rotation=<rot 0.0001 0 -100.1445>}]
 dyn_jacket_front_r_03 [{
   shapeType=Capsule
   length=13.875
   radius=5.3125
   shapeName=dyn_jacket_front_l_03_capsule
   center=<-3 0 -6>
   rotation=<rot 0.0003 0.0001 -97.2798>
  }]
 dyn_jacket_front_r_04 [{
   shapeType=Capsule
   length=11.0625
   radius=7.3125
   shapeName=dyn_jacket_front_l_04_capsule
   center=<-3 0 -6>
   rotation=<rot 0.0001 0 -102.2345>
  }]
 dyn_jacket_front_r_05 [{
   shapeType=Capsule
   length=14.5625
   radius=7.6875
   shapeName=dyn_jacket_front_l_05_capsule
   center=<-6.4606 0.5341 -7.5755>
   rotation=<rot 0.0003 0.0001 -108.7267>
  }]
 dyn_jacket_front_r_MSTR [{shapeType=Capsule, length=0.505, radius=0.505, shapeName=dyn_jacket_front_r_MSTR_capsule, rotation=<rot 0 -90 0>}]
 dyn_jacket_rear_01 [{shapeType=Capsule, length=12.5625, radius=3.1875, shapeName=dyn_jacket_rear_01_capsule, center=<3 0 5>, rotation=<rot -0.0004 0 96.5003>}]
 dyn_jacket_rear_02 [{shapeType=Capsule, length=17.1875, radius=3.75, shapeName=dyn_jacket_rear_02_capsule, center=<3 0 5>, rotation=<rot -0.0004 0 96.5003>}]
 dyn_jacket_rear_03 [{shapeType=Capsule, length=16.0625, radius=5, shapeName=dyn_jacket_rear_03_capsule, center=<3 0 5>, rotation=<rot -0.0004 0 96.5003>}]
 dyn_jacket_rear_04 [{shapeType=Capsule, length=17.9375, radius=7.25, shapeName=dyn_jacket_rear_04_capsule, center=<3 0 5>, rotation=<rot -0.0004 0 96.5003>}]
 dyn_jacket_rear_MSTR [{shapeType=Capsule, length=0.505, radius=0.505, shapeName=dyn_jacket_rear_MSTR_capsule, rotation=<rot 0 -90 0>}]
 dyn_jacket_rear_l_01 [{
   shapeType=Capsule
   length=9.0625
   radius=5.0625
   shapeName=dyn_jacket_rear_l_01_capsule
   center=<2.9998 1 5.5>
   rotation=<rot -0.0208 0.0098 94.6331>
  }]
 dyn_jacket_rear_l_02 [{
   shapeType=Capsule
   length=9.4375
   radius=5.75
   shapeName=dyn_jacket_rear_l_02_capsule
   center=<2.9998 1 5.5>
   rotation=<rot -0.0208 0.0097 94.6328>
  }]
 dyn_jacket_rear_l_03 [{
   shapeType=Capsule
   length=11.8125
   radius=7.3125
   shapeName=dyn_jacket_rear_l_03_capsule
   center=<2.9998 1 5.5>
   rotation=<rot -0.0208 0.0097 94.6325>
  }]
 dyn_jacket_rear_l_04 [{
   shapeType=Capsule
   length=11.25
   radius=7.875
   shapeName=dyn_jacket_rear_l_04_capsule
   center=<2.9998 1 5.5>
   rotation=<rot -0.0208 0.0097 94.6325>
  }]
 dyn_jacket_rear_l_MSTR [{shapeType=Capsule, length=0.505, radius=0.505, shapeName=dyn_jacket_rear_l_MSTR_capsule, rotation=<rot 0 -90 0>}]
 dyn_jacket_rear_r_01 [{
   shapeType=Capsule
   length=9.0625
   radius=5.0625
   shapeName=dyn_jacket_rear_l_01_capsule
   center=<-2.9998 -1 -5.5>
   rotation=<rot -0.0208 0.0098 -85.3669>
  }]
 dyn_jacket_rear_r_02 [{
   shapeType=Capsule
   length=9.4375
   radius=5.75
   shapeName=dyn_jacket_rear_l_02_capsule
   center=<-2.9998 -1 -5.5>
   rotation=<rot -0.0208 0.0097 -85.3672>
  }]
 dyn_jacket_rear_r_03 [{
   shapeType=Capsule
   length=12.3125
   radius=6.8125
   shapeName=dyn_jacket_rear_l_03_capsule
   center=<-2.9998 -1 -5.5>
   rotation=<rot -0.0208 0.0097 -85.3675>
  }]
 dyn_jacket_rear_r_04 [{
   shapeType=Capsule
   length=11.75
   radius=7.375
   shapeName=dyn_jacket_rear_l_04_capsule
   center=<-2.9998 -1 -5.5>
   rotation=<rot -0.0208 0.0097 -85.3675>
  }]
 dyn_jacket_rear_r_MSTR [{shapeType=Capsule, length=0.505, radius=0.505, shapeName=dyn_jacket_rear_r_MSTR_capsule, rotation=<rot 0 -90 0>}]
 dyn_jacket_side_l_01 [{shapeType=Capsule, length=15.4375, radius=4.9375, shapeName=dyn_jacket_side_l_01_capsule, center=<3 0 6>, rotation=<rot 0 0.0003 89.7126>}]
 dyn_jacket_side_l_02 [{shapeType=Capsule, length=15.8125, radius=5.4375, shapeName=dyn_jacket_side_l_02_capsule, center=<3 0 6>, rotation=<rot 0 0.0005 91.4975>}]
 dyn_jacket_side_l_03 [{shapeType=Capsule, length=20.125, radius=5.625, shapeName=dyn_jacket_side_l_03_capsule, center=<3 0 6>, rotation=<rot 0 0 90>}]
 dyn_jacket_side_l_04 [{shapeType=Capsule, length=21.625, radius=6.75, shapeName=dyn_jacket_side_l_04_capsule, center=<3 0 5>, rotation=<rot 0 0 90>}]
 dyn_jacket_side_l_MSTR [{shapeType=Capsule, length=0.505, radius=0.505, shapeName=dyn_jacket_side_l_MSTR_capsule, rotation=<rot 0 -90 0>}]
 dyn_jacket_side_r_01 [{
   shapeType=Capsule
   length=15.4375
   radius=4.9375
   shapeName=dyn_jacket_side_l_01_capsule
   center=<-3 -3 -6.0132>
   rotation=<rot 0 0.0003 -90.2874>
  }]
 dyn_jacket_side_r_02 [{
   shapeType=Capsule
   length=15.8125
   radius=5.4375
   shapeName=dyn_jacket_side_l_02_capsule
   center=<-3 -3 -5.9312>
   rotation=<rot 0 0.0005 -88.5025>
  }]
 dyn_jacket_side_r_03 [{shapeType=Capsule, length=20.125, radius=5.625, shapeName=dyn_jacket_side_l_03_capsule, center=<-3 -3 -6>, rotation=<rot 0 0 -90>}]
 dyn_jacket_side_r_04 [{shapeType=Capsule, length=21.625, radius=6.75, shapeName=dyn_jacket_side_l_04_capsule, center=<-3 -3 -5>, rotation=<rot 0 0 -90>}]
 dyn_jacket_side_r_MSTR [{shapeType=Capsule, length=0.505, radius=0.505, shapeName=dyn_jacket_side_r_MSTR_capsule, rotation=<rot 0 -90 0>}]
 pelvis [
  {shapeType=Capsule, length=14.4375, radius=12.5, center=<-4.9153 -0.6893 0>}
  {shapeType=Capsule, length=12.125, radius=14.5625, shapeName=pelvis_capsule, center=<-12.3424 0 0>}
  {shapeType=Capsule, length=46.4375, radius=15, shapeName=pelvis_capsule, center=<0 -45.8775 0>}
 ]
 spine_01 [{shapeType=Capsule, length=13.875, radius=12.25, shapeName=spine_01_capsule, center=<-3.3307 -0.9059 0>, rotation=<rot 0 -90 0>}]
 spine_02 [{shapeType=Capsule, length=10, radius=13.25, center=<-2.3936 -0.5821 0>}]
 spine_03 [{shapeType=Capsule, length=9.4375, radius=13.8125, center=<0 -0.2927 0>}]
 spine_04 [{shapeType=Capsule, length=12.125, radius=11.4375, center=<0 0.5624 0>}]
 thigh_l [{
   shapeType=Capsule
   length=38.9043
   radius=9.9134
   shapeName=thigh_l_capsule
   center=<-24.8341 0.0468 0.9536>
   rotation=<rot 86.1268 70.6195 64.9303>
  }]
 thigh_r [{
   shapeType=Capsule
   length=38.9043
   radius=9.9134
   shapeName=thigh_l_capsule
   center=<24.8341 -0.0468 -0.9536>
   rotation=<rot 86.1268 70.6195 -115.0697>
  }]
}
constraints [
 {bone1Name=spine_02, bone2Name=spine_01, swing1LimitDegrees=45, swing2LimitDegrees=45, twistLimitDegrees=45}
 {bone1Name=spine_03, bone2Name=spine_02, swing1LimitDegrees=45, swing2LimitDegrees=45, twistLimitDegrees=45}
 {bone1Name=spine_04, bone2Name=spine_03, swing1LimitDegrees=45, swing2LimitDegrees=45, twistLimitDegrees=45}
 {
  bone1Name=dyn_jacket_front_l_MSTR
  bone2Name=spine_04
  swing1LimitDegrees=45
  swing1Motion=Free
  swing2LimitDegrees=45
  swing2Motion=Free
  twistLimitDegrees=45
  twistMotion=Free
 }
 {bone1Name=dyn_jacket_front_l_01, bone2Name=dyn_jacket_front_l_MSTR, swing1LimitDegrees=6, swing2LimitDegrees=14, twistLimitDegrees=12}
 {bone1Name=dyn_jacket_front_l_02, bone2Name=dyn_jacket_front_l_01, swing1LimitDegrees=6, swing2LimitDegrees=6, twistLimitDegrees=12}
 {bone1Name=dyn_jacket_front_l_03, bone2Name=dyn_jacket_front_l_02, swing1LimitDegrees=6, swing2LimitDegrees=6, twistLimitDegrees=8}
 {bone1Name=dyn_jacket_front_l_04, bone2Name=dyn_jacket_front_l_03, swing1LimitDegrees=6, swing2LimitDegrees=6, twistLimitDegrees=8}
 {bone1Name=dyn_jacket_front_l_05, bone2Name=dyn_jacket_front_l_04, swing1LimitDegrees=6, swing2LimitDegrees=6, twistLimitDegrees=8}
 {
  bone1Name=dyn_jacket_front_r_MSTR
  bone2Name=spine_04
  swing1LimitDegrees=45
  swing1Motion=Free
  swing2LimitDegrees=45
  swing2Motion=Free
  twistLimitDegrees=45
  twistMotion=Free
 }
 {
  bone1Name=dyn_jacket_rear_MSTR
  bone2Name=spine_04
  swing1LimitDegrees=45
  swing1Motion=Free
  swing2LimitDegrees=45
  swing2Motion=Free
  twistLimitDegrees=45
  twistMotion=Free
 }
 {bone1Name=dyn_jacket_rear_01, bone2Name=dyn_jacket_rear_MSTR, swing1LimitDegrees=6, swing2LimitDegrees=6, twistLimitDegrees=2}
 {bone1Name=dyn_jacket_rear_02, bone2Name=dyn_jacket_rear_01, swing1LimitDegrees=6, swing2LimitDegrees=6, twistLimitDegrees=6}
 {bone1Name=dyn_jacket_rear_03, bone2Name=dyn_jacket_rear_02, swing1LimitDegrees=6, swing2LimitDegrees=6, twistLimitDegrees=12}
 {bone1Name=dyn_jacket_rear_04, bone2Name=dyn_jacket_rear_03, swing1LimitDegrees=6, swing2LimitDegrees=6, twistLimitDegrees=12}
 {
  bone1Name=dyn_jacket_rear_l_MSTR
  bone2Name=spine_04
  swing1LimitDegrees=45
  swing1Motion=Free
  swing2LimitDegrees=45
  swing2Motion=Free
  twistLimitDegrees=45
  twistMotion=Free
 }
 {
  bone1Name=dyn_jacket_rear_r_MSTR
  bone2Name=spine_04
  swing1LimitDegrees=45
  swing1Motion=Free
  swing2LimitDegrees=45
  swing2Motion=Free
  twistLimitDegrees=45
  twistMotion=Free
 }
 {
  bone1Name=dyn_jacket_side_l_MSTR
  bone2Name=spine_04
  swing1LimitDegrees=45
  swing1Motion=Free
  swing2LimitDegrees=45
  swing2Motion=Free
  twistLimitDegrees=45
  twistMotion=Free
 }
 {bone1Name=dyn_jacket_side_l_01, bone2Name=dyn_jacket_side_l_MSTR, swing1LimitDegrees=6, swing2LimitDegrees=4, twistLimitDegrees=4}
 {bone1Name=dyn_jacket_side_l_02, bone2Name=dyn_jacket_side_l_01, swing1LimitDegrees=6, swing2LimitDegrees=4, twistLimitDegrees=4}
 {bone1Name=dyn_jacket_side_l_03, bone2Name=dyn_jacket_side_l_02, swing1LimitDegrees=6, swing2LimitDegrees=4, twistLimitDegrees=4}
 {bone1Name=dyn_jacket_side_l_04, bone2Name=dyn_jacket_side_l_03, swing1LimitDegrees=6, swing2LimitDegrees=4, twistLimitDegrees=4}
 {
  bone1Name=dyn_jacket_side_r_MSTR
  bone2Name=spine_04
  swing1LimitDegrees=45
  swing1Motion=Free
  swing2LimitDegrees=45
  swing2Motion=Free
  twistLimitDegrees=45
  twistMotion=Free
 }
 {
  bone1Name=thigh_l
  bone2Name=pelvis
  swing1LimitDegrees=45
  swing1Motion=Free
  swing2LimitDegrees=45
  swing2Motion=Free
  twistLimitDegrees=45
  twistMotion=Free
 }
 {
  bone1Name=spine_01
  bone2Name=pelvis
  swing1LimitDegrees=45
  swing1Motion=Free
  swing2LimitDegrees=45
  swing2Motion=Free
  twistLimitDegrees=45
  twistMotion=Free
 }
 {bone1Name=dyn_jacket_rear_l_01, bone2Name=dyn_jacket_rear_l_MSTR, swing1LimitDegrees=6, swing2LimitDegrees=6, twistLimitDegrees=6}
 {bone1Name=dyn_jacket_rear_l_02, bone2Name=dyn_jacket_rear_l_01, swing1LimitDegrees=6, swing2LimitDegrees=6, twistLimitDegrees=12}
 {bone1Name=dyn_jacket_rear_l_03, bone2Name=dyn_jacket_rear_l_02, swing1LimitDegrees=6, swing2LimitDegrees=6, twistLimitDegrees=6}
 {bone1Name=dyn_jacket_rear_l_04, bone2Name=dyn_jacket_rear_l_03, swing1LimitDegrees=6, swing2LimitDegrees=6, twistLimitDegrees=12}
 {
  bone1Name=dyn_jacket_rear_02
  bone2Name=dyn_jacket_rear_l_02
  swing1LimitDegrees=45
  swing1Motion=Free
  swing2LimitDegrees=45
  swing2Motion=Free
  twistLimitDegrees=45
  twistMotion=Free
 }
 {
  bone1Name=dyn_jacket_rear_04
  bone2Name=dyn_jacket_rear_l_04
  swing1LimitDegrees=45
  swing1Motion=Free
  swing2LimitDegrees=45
  swing2Motion=Free
  twistLimitDegrees=45
  twistMotion=Free
 }
 {bone1Name=dyn_jacket_rear_r_01, bone2Name=dyn_jacket_rear_r_MSTR, swing1LimitDegrees=6, swing2LimitDegrees=6, twistLimitDegrees=12}
 {bone1Name=dyn_jacket_rear_r_02, bone2Name=dyn_jacket_rear_r_01, swing1LimitDegrees=6, swing2LimitDegrees=6, twistLimitDegrees=6}
 {bone1Name=dyn_jacket_rear_r_03, bone2Name=dyn_jacket_rear_r_02, swing1LimitDegrees=6, swing2LimitDegrees=6, twistLimitDegrees=12}
 {bone1Name=dyn_jacket_rear_r_04, bone2Name=dyn_jacket_rear_r_03, swing1LimitDegrees=6, swing2LimitDegrees=6, twistLimitDegrees=2}
 {
  bone1Name=dyn_jacket_rear_02
  bone2Name=dyn_jacket_rear_r_02
  swing1LimitDegrees=45
  swing1Motion=Free
  swing2LimitDegrees=45
  swing2Motion=Free
  twistLimitDegrees=45
  twistMotion=Free
 }
 {
  bone1Name=dyn_jacket_rear_03
  bone2Name=dyn_jacket_rear_r_03
  swing1LimitDegrees=45
  swing1Motion=Free
  swing2LimitDegrees=45
  swing2Motion=Free
  twistLimitDegrees=45
  twistMotion=Free
 }
 {
  bone1Name=dyn_jacket_rear_04
  bone2Name=dyn_jacket_rear_r_04
  swing1LimitDegrees=45
  swing1Motion=Free
  swing2LimitDegrees=45
  swing2Motion=Free
  twistLimitDegrees=45
  twistMotion=Free
 }
 {
  bone1Name=thigh_r
  bone2Name=pelvis
  swing1LimitDegrees=45
  swing1Motion=Free
  swing2LimitDegrees=45
  swing2Motion=Free
  twistLimitDegrees=45
  twistMotion=Free
 }
 {bone1Name=dyn_jacket_side_r_01, bone2Name=dyn_jacket_side_r_MSTR, swing1LimitDegrees=6, swing2LimitDegrees=4, twistLimitDegrees=4}
 {bone1Name=dyn_jacket_side_r_02, bone2Name=dyn_jacket_side_r_01, swing1LimitDegrees=6, swing2LimitDegrees=4, twistLimitDegrees=4}
 {bone1Name=dyn_jacket_side_r_03, bone2Name=dyn_jacket_side_r_02, swing1LimitDegrees=6, swing2LimitDegrees=4, twistLimitDegrees=4}
 {bone1Name=dyn_jacket_side_r_04, bone2Name=dyn_jacket_side_r_03, swing1LimitDegrees=6, swing2LimitDegrees=4, twistLimitDegrees=4}
 {bone1Name=dyn_jacket_front_r_01, bone2Name=dyn_jacket_front_r_MSTR, swing1LimitDegrees=6, swing2LimitDegrees=14, twistLimitDegrees=12}
 {bone1Name=dyn_jacket_front_r_02, bone2Name=dyn_jacket_front_r_01, swing1LimitDegrees=6, swing2LimitDegrees=14, twistLimitDegrees=12}
 {bone1Name=dyn_jacket_front_r_03, bone2Name=dyn_jacket_front_r_02, swing1LimitDegrees=6, swing2LimitDegrees=6, twistLimitDegrees=8}
 {bone1Name=dyn_jacket_front_r_04, bone2Name=dyn_jacket_front_r_03, swing1LimitDegrees=6, swing2LimitDegrees=6, twistLimitDegrees=8}
 {bone1Name=dyn_jacket_front_r_05, bone2Name=dyn_jacket_front_r_04, swing1LimitDegrees=6, swing2LimitDegrees=6, twistLimitDegrees=8}
 {
  bone1Name=dyn_jacket_rear_l_04
  bone2Name=dyn_jacket_rear_04
  swing1LimitDegrees=45
  swing1Motion=Free
  swing2LimitDegrees=45
  swing2Motion=Free
  twistLimitDegrees=45
  twistMotion=Free
 }
 {
  bone1Name=dyn_jacket_rear_r_04
  bone2Name=dyn_jacket_rear_04
  swing1LimitDegrees=45
  swing1Motion=Free
  swing2LimitDegrees=45
  swing2Motion=Free
  twistLimitDegrees=45
  twistMotion=Free
 }
 {
  bone1Name=dyn_jacket_rear_r_03
  bone2Name=dyn_jacket_rear_03
  swing1LimitDegrees=45
  swing1Motion=Free
  swing2LimitDegrees=45
  swing2Motion=Free
  twistLimitDegrees=45
  twistMotion=Free
 }
 {
  bone1Name=dyn_jacket_rear_l_03
  bone2Name=dyn_jacket_rear_03
  swing1LimitDegrees=45
  swing1Motion=Free
  swing2LimitDegrees=45
  swing2Motion=Free
  twistLimitDegrees=45
  twistMotion=Free
 }
 {
  bone1Name=dyn_jacket_front_l_04
  bone2Name=dyn_jacket_side_l_03
  swing1LimitDegrees=45
  swing1Motion=Free
  swing2LimitDegrees=45
  swing2Motion=Free
  twistLimitDegrees=45
  twistMotion=Free
 }
 {
  bone1Name=dyn_jacket_front_r_04
  bone2Name=dyn_jacket_side_r_03
  swing1LimitDegrees=45
  swing1Motion=Free
  swing2LimitDegrees=45
  swing2Motion=Free
  twistLimitDegrees=45
  twistMotion=Free
 }
]
