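# Character-part AnimBP for the LemonCartGranite male body cosmetic, driving jacket and necklace secondary motion atop the parent mesh's pose.
# A Control Rig and backbling-weighted collider constraints feed Chaos cloth when enabled, otherwise RigidBody simulation with backpack-specific physics and additive bone corrections.
# AnimDynamics then swings the necklace, tuned separately for front end and gameplay.
# The EventGraph updates cloth or gravity settings and material WPO, disabling cloth in first person.
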AnimBlueprint /BRCosmetics/Characters/Player/Male/Medium/Bodies/M_MED_LemonCartGranite/Meshes/M_MED_LemonCartGranite_AnimBP
target_skeleton=&/M_MED_LemonCartGranite_Skeleton
graphs [
 AnimGraph {nodes [
   Output Pose(Result=Use cached pose 'Post Necklace Dynamics'()) {layerGroup=None}
   n1 = "Constraint - dyn_backpack_collider (P)"(ComponentPose=Local To Component(LocalPose=Copy Pose From Mesh() {bUseAttachedParent=true}), Alpha=1, ConstraintWeights_0=1) {
    alphaBoolBlend {blendOption=Linear}
    boneToModify {boneName=dyn_backpack_collider}
    constraintSetup [{targetBone {boneName=attach_backpack}}]
    constraintWeights [1]
   }
   n2 = "Constraint - dyn_cloth_jacket_collision_spine_02 (P)"(ComponentPose=@n1, Alpha=1, ConstraintWeights_0=1) {
    alphaBoolBlend {blendOption=Linear}
    boneToModify {boneName=dyn_cloth_jacket_collision_spine_02}
    constraintSetup [{offsetOption=None, targetBone {boneName=spine_01}}]
    constraintWeights [1]
   }
   n3 = "Constraint - dyn_cloth_jacket_collision_spine_01 (P)"(ComponentPose=@n2, Alpha=1, ConstraintWeights_0=1) {
    alphaBoolBlend {blendOption=Linear}
    boneToModify {boneName=dyn_cloth_jacket_collision_spine_01}
    constraintSetup [{offsetOption=None, targetBone {boneName=pelvis}}]
    constraintWeights [1]
   }
   n4 = "Constraint - dyn_backpack_collider (P)"(ComponentPose=Local To Component(LocalPose=Copy Pose From Mesh() {bUseAttachedParent=true}), Alpha=1, ConstraintWeights_0=1) {
    alphaBoolBlend {blendOption=Linear}
    boneToModify {boneName=dyn_backpack_collider}
    constraintSetup [{offsetOption=None, targetBone {boneName=root}}]
    constraintWeights [1]
   }
   n5 = "Constraint - dyn_cloth_jacket_collision_spine_02 (P)"(ComponentPose=@n4, Alpha=1, ConstraintWeights_0=1) {
    alphaBoolBlend {blendOption=Linear}
    boneToModify {boneName=dyn_cloth_jacket_collision_spine_02}
    constraintSetup [{offsetOption=None, targetBone {boneName=root}}]
    constraintWeights [1]
   }
   n6 = "Constraint - dyn_cloth_jacket_collision_spine_01 (P)"(ComponentPose=@n5, Alpha=1, ConstraintWeights_0=1) {
    alphaBoolBlend {blendOption=Linear}
    boneToModify {boneName=dyn_cloth_jacket_collision_spine_01}
    constraintSetup [{offsetOption=None, targetBone {boneName=root}}]
    constraintWeights [1]
   }
   n7 = "Transform (Modify) Bone - Bone: dyn_cloth_jacket_collision_spine_02"!(ComponentPose=@n6, Alpha=1, Translation="0,0,-100") -> (Pose) {
    translationMode=BMM_Additive
    translationSpace=BCS_BoneSpace
    alphaBoolBlend {blendOption=Linear}
    boneToModify {boneName=dyn_cloth_jacket_collision_spine_02}
    translation {z=-100}
   }
   n8 = "Transform (Modify) Bone - Bone: dyn_cloth_jacket_collision_spine_01"!(ComponentPose=@n7, Alpha=1, Translation="0,0,-100") -> (Pose) {
    translationMode=BMM_Additive
    translationSpace=BCS_BoneSpace
    alphaBoolBlend {blendOption=Linear}
    boneToModify {boneName=dyn_cloth_jacket_collision_spine_01}
    translation {z=-100}
   }
   n9 = Control Rig!(Source=Copy Pose From Mesh() {bUseAttachedParent=true}, Alpha, Disable RBAN Jacket Deformations:bool=Get bClothEnabled()) -> (Pose) {
    bTransferInputCurves=false
    alphaBoolBlend {blendOption=Linear}
    controlRigAssetReference {blueprintRigClass=&/M_MED_LemonCartGranite_CtrlRig.M_MED_LemonCartGranite_CtrlRig_C}
   }
   n10 = Layered blend per bone!(
    BasePose=@n9
    BlendPoses_0=Component To Local(ComponentPose=@n3)
    BlendPoses_1:Pose Link Structure=Component To Local(ComponentPose=@n8)
    BlendWeights_0=Get Wearing Backbling Float() -> (Wearing Backbling Float:double)
    BlendWeights_1:float=Get Not Wearing Backbling Inverse Float() -> (Not Wearing Backbling Inverse Float:double)
   ) -> (Pose) {
    blendPoses [{linkId=-1, sourceLinkId=-1}, {linkId=-1, sourceLinkId=-1}]
    blendWeights [1,1]
    layerSetup [
     {branchFilters [{boneName=dyn_backpack_collider}, {boneName=dyn_cloth_jacket_collision_spine_01}, {boneName=dyn_cloth_jacket_collision_spine_02}]}
     {branchFilters [{boneName=dyn_backpack_collider}, {boneName=dyn_cloth_jacket_collision_spine_01}, {boneName=dyn_cloth_jacket_collision_spine_02}]}
    ]
   }
   Save cached pose 'Post Deformation'(Pose=@n10) {cacheName=Post Deformation}
   Save cached pose 'Post Jacket Dynamics'(Pose=Blend Poses by bool(
     BlendPose_0=Use cached pose 'Post Deformation'()
     BlendPose_1=Use cached pose 'Post RBAN'()
     BlendTime_0=0
     BlendTime_1=0
     bActiveValue=Get bClothEnabled()
    ) {blendType=Linear}) {cacheName=Post Jacket Dynamics}
   n11 = RigidBody!(
    ComponentPose=Local To Component(LocalPose=Use cached pose 'Post Deformation'())
    Alpha
    OverrideWorldGravity:Vector="0,0,-980"
    ExternalForce
    ComponentLinearAccScale:Vector="0,0,0"
    ComponentLinearVelScale:Vector="0,0,0"
    SimSpaceSettings:Sim Space Settings Structure=(WorldAlpha=0,VelocityScaleZ=1,DampingAlpha=1,MaxLinearVelocity=10000,MaxAngularVelocity=10000,MaxLinearAcceleration=10000,MaxAngularAcceleration=10000,ExternalLinearDragV=(X=0,Y=0,Z=0),ExternalLinearVelocity=(X=0,Y=0,Z=0),ExternalAngularVelocity=(X=0,Y=0,Z=0))
   ) -> (Pose) {
    bOverrideWorldGravity=true
    overridePhysicsAsset=&/M_MED_LemonCartGranite_JacketBackpack_Physics
    simulationSpace=BaseBoneSpace
    alphaBoolBlend {blendOption=Linear}
    baseBoneRef {boneName=pelvis}
    componentAppliedLinearAccClamp=<0 0 0>
    componentLinearAccScale=<0.3 0.3 0.3>
    componentLinearVelScale=<0.3 0.3 0.3>
    overrideWorldGravity {z=-980}
   }
   n12 = Component To Local(ComponentPose=@n11)
   n13 = RigidBody(
    ComponentPose=Local To Component(LocalPose=Use cached pose 'Post Deformation'())
    OverrideWorldGravity="0,0,-980"
    ComponentLinearAccScale="0,0,0"
    ComponentLinearVelScale="0,0,0"
    SimSpaceSettings=(WorldAlpha=0,VelocityScaleZ=1,DampingAlpha=1,MaxLinearVelocity=10000,MaxAngularVelocity=10000,MaxLinearAcceleration=10000,MaxAngularAcceleration=10000,ExternalLinearDragV=(X=0,Y=0,Z=0),ExternalLinearVelocity=(X=0,Y=0,Z=0),ExternalAngularVelocity=(X=0,Y=0,Z=0))
   ) {
    bOverrideWorldGravity=true
    overridePhysicsAsset=&/M_MED_LemonCartGranite_Physics
    simulationSpace=BaseBoneSpace
    alphaBoolBlend {blendOption=Linear}
    baseBoneRef {boneName=spine_05}
    componentAppliedLinearAccClamp=<0 0 0>
    overrideWorldGravity {z=-980}
   }
   n14 = Component To Local(ComponentPose=@n13)
   n15 = Get Wearing Backbling() -> (Wearing Backbling:bool)
   n16 = "Transform (Modify) Bone - Bone: dyn_jacket_front_l_MSTR"!(
    ComponentPose=Local To Component(LocalPose=Blend Poses by bool(BlendPose_0=@n12, BlendPose_1=@n14, BlendTime_0=0, BlendTime_1=0, bActiveValue=@n15) {blendType=Linear})
    Alpha=1
    Translation="0,0,0.5"
    Scale="1,1,1"
   ) -> (Pose) {
    rotationMode=BMM_Additive
    rotationSpace=BCS_BoneSpace
    translationMode=BMM_Additive
    translationSpace=BCS_BoneSpace
    alphaBoolBlend {blendOption=Linear}
    boneToModify {boneName=dyn_jacket_front_l_MSTR}
    translation {z=0.5}
   }
   n17 = "Transform (Modify) Bone - Bone: dyn_jacket_side_l_MSTR"!(ComponentPose=@n16, Alpha=1, Translation="0,0,1", Scale="1,1,1") -> (Pose) {
    rotationMode=BMM_Additive
    rotationSpace=BCS_BoneSpace
    translationMode=BMM_Additive
    translationSpace=BCS_BoneSpace
    alphaBoolBlend {blendOption=Linear}
    boneToModify {boneName=dyn_jacket_side_l_MSTR}
    translation {z=1}
   }
   n18 = "Transform (Modify) Bone - Bone: dyn_jacket_side_l_MSTR"!(ComponentPose=@n17, bAlphaBoolEnabled:bool=True, Translation="0,0,0", Scale="1,1,1") -> (Pose) {
    alphaInputType=Bool
    rotationMode=BMM_Additive
    rotationSpace=BCS_BoneSpace
    translationMode=BMM_Additive
    translationSpace=BCS_BoneSpace
    alphaBoolBlend {blendInTime=0.2, blendOption=Linear, blendOutTime=0.2}
    boneToModify {boneName=dyn_jacket_side_l_MSTR}
    rotation {pitch=10}
   }
   n19 = "Transform (Modify) Bone - Bone: dyn_jacket_rear_l_MSTR"(ComponentPose=@n18, Alpha=1, Translation="0,0,2", Rotation="0,0,0", Scale="1,1,1") {
    rotationMode=BMM_Additive
    rotationSpace=BCS_BoneSpace
    translationMode=BMM_Additive
    translationSpace=BCS_BoneSpace
    alphaBoolBlend {blendOption=Linear}
    boneToModify {boneName=dyn_jacket_rear_l_MSTR}
    translation {z=2}
   }
   n20 = "Transform (Modify) Bone - Bone: dyn_jacket_rear_l_MSTR"!(ComponentPose=@n19, bAlphaBoolEnabled:bool=True, Translation="0,0,0", Rotation="10,0,0", Scale="1,1,1") -> (Pose) {
    alphaInputType=Bool
    rotationMode=BMM_Additive
    rotationSpace=BCS_BoneSpace
    translationMode=BMM_Additive
    translationSpace=BCS_BoneSpace
    alphaBoolBlend {blendInTime=0.2, blendOption=Linear, blendOutTime=0.2}
    boneToModify {boneName=dyn_jacket_rear_l_MSTR}
    rotation {pitch=10}
   }
   n21 = "Transform (Modify) Bone - Bone: dyn_jacket_rear_l_02"!(ComponentPose=@n20, Alpha=1, Translation="0,0,0", Scale="1,1,1") -> (Pose) {
    rotationMode=BMM_Additive
    rotationSpace=BCS_BoneSpace
    translationMode=BMM_Additive
    translationSpace=BCS_BoneSpace
    alphaBoolBlend {blendOption=Linear}
    boneToModify {boneName=dyn_jacket_rear_l_02}
    translation=<0 0>
   }
   n22 = "Transform (Modify) Bone - Bone: dyn_jacket_rear_MSTR"(ComponentPose=@n21, Alpha=1, Translation="0,0,3", Rotation="0,0,0", Scale="1,1,1") {
    rotationMode=BMM_Additive
    rotationSpace=BCS_BoneSpace
    translationMode=BMM_Additive
    translationSpace=BCS_BoneSpace
    alphaBoolBlend {blendOption=Linear}
    boneToModify {boneName=dyn_jacket_rear_MSTR}
    translation {z=3}
   }
   n23 = "Transform (Modify) Bone - Bone: dyn_jacket_front_r_MSTR"!(ComponentPose=@n22, Alpha=1, Translation="0,0,-0.5", Scale="1,1,1") -> (Pose) {
    rotationMode=BMM_Additive
    rotationSpace=BCS_BoneSpace
    translationMode=BMM_Additive
    translationSpace=BCS_BoneSpace
    alphaBoolBlend {blendOption=Linear}
    boneToModify {boneName=dyn_jacket_front_r_MSTR}
    translation {z=-0.5}
   }
   n24 = "Transform (Modify) Bone - Bone: dyn_jacket_side_r_MSTR"!(ComponentPose=@n23, Alpha=1, Translation="0,0,-0.5", Scale="1,1,1") -> (Pose) {
    rotationMode=BMM_Additive
    rotationSpace=BCS_BoneSpace
    translationMode=BMM_Additive
    translationSpace=BCS_BoneSpace
    alphaBoolBlend {blendOption=Linear}
    boneToModify {boneName=dyn_jacket_side_r_MSTR}
    translation {z=-0.5}
   }
   n25 = "Transform (Modify) Bone - Bone: dyn_jacket_rear_r_MSTR"(ComponentPose=@n24, Alpha=1, Translation="-0.081,0.000,-2.720", Rotation="-6.5,0,0", Scale="1,1,1") {
    rotationMode=BMM_Additive
    rotationSpace=BCS_BoneSpace
    translationMode=BMM_Additive
    translationSpace=BCS_BoneSpace
    alphaBoolBlend {blendOption=Linear}
    boneToModify {boneName=dyn_jacket_rear_r_MSTR}
    rotation {pitch=-6.5}
    translation {x=-0.081, z=-2.72}
   }
   n26 = "Transform (Modify) Bone - Bone: dyn_jacket_rear_r_02"!(ComponentPose=@n25, Alpha=1, Translation="0,0,0", Scale="1,1,1") -> (Pose) {
    rotationMode=BMM_Additive
    rotationSpace=BCS_BoneSpace
    translationMode=BMM_Additive
    translationSpace=BCS_BoneSpace
    alphaBoolBlend {blendOption=Linear}
    boneToModify {boneName=dyn_jacket_rear_r_02}
   }
   n27 = "Transform (Modify) Bone - Bone: dyn_jacket_side_l_MSTR"(ComponentPose=@n26, bAlphaBoolEnabled=True, Translation="0,0,0", Scale="1,1,1") {
    alphaInputType=Bool
    rotationMode=BMM_Additive
    rotationSpace=BCS_BoneSpace
    translationMode=BMM_Additive
    translationSpace=BCS_BoneSpace
    alphaBoolBlend {blendOption=Linear}
    boneToModify {boneName=dyn_jacket_side_l_MSTR}
   }
   n28 = "Transform (Modify) Bone - Bone: dyn_jacket_rear_l_MSTR"(ComponentPose=@n27, bAlphaBoolEnabled=True, Translation="0,0,0.7", Rotation="2.5,0,0", Scale="1,1,1") {
    alphaInputType=Bool
    rotationMode=BMM_Additive
    rotationSpace=BCS_BoneSpace
    translationMode=BMM_Additive
    translationSpace=BCS_BoneSpace
    alphaBoolBlend {blendOption=Linear}
    boneToModify {boneName=dyn_jacket_rear_l_MSTR}
    rotation {pitch=2.5}
    translation {z=0.7}
   }
   n29 = "Transform (Modify) Bone - Bone: dyn_jacket_rear_MSTR"!(ComponentPose=@n28, bAlphaBoolEnabled:bool=True, Translation="0,0,0", Rotation="2.5,0,0", Scale="1,1,1") -> (Pose) {
    alphaInputType=Bool
    rotationMode=BMM_Additive
    rotationSpace=BCS_BoneSpace
    translationMode=BMM_Additive
    translationSpace=BCS_BoneSpace
    alphaBoolBlend {blendOption=Linear}
    boneToModify {boneName=dyn_jacket_rear_MSTR}
    rotation {pitch=2.5}
   }
   n30 = "Transform (Modify) Bone - Bone: dyn_jacket_side_r_MSTR"!(ComponentPose=@n29, bAlphaBoolEnabled:bool=True, Translation="0,0,0", Scale="1,1,1") -> (Pose) {
    alphaInputType=Bool
    rotationMode=BMM_Additive
    rotationSpace=BCS_BoneSpace
    translationMode=BMM_Additive
    translationSpace=BCS_BoneSpace
    alphaBoolBlend {blendOption=Linear}
    boneToModify {boneName=dyn_jacket_side_r_MSTR}
   }
   n31 = "Transform (Modify) Bone - Bone: dyn_jacket_rear_r_MSTR"!(ComponentPose=@n30, bAlphaBoolEnabled:bool=True, Translation="0,0,-0.7", Rotation="2.5,0,0", Scale="1,1,1") -> (Pose) {
    alphaInputType=Bool
    rotationMode=BMM_Additive
    rotationSpace=BCS_BoneSpace
    translationMode=BMM_Additive
    translationSpace=BCS_BoneSpace
    alphaBoolBlend {blendOption=Linear}
    boneToModify {boneName=dyn_jacket_rear_r_MSTR}
    rotation {pitch=2.5}
    translation {z=-0.7}
   }
   Save cached pose 'Post RBAN'(Pose=Component To Local(ComponentPose=@n31)) {cacheName=Post RBAN}
   n32 = Anim Dynamics - Bone: dyn_necklace_01(ComponentPose=Local To Component(LocalPose=Use cached pose 'Post Jacket Dynamics'()), Alpha=1, ExternalForce="0,0,0") {
    angularDampingOverride=1
    angularSpringConstant=20000
    bAngularSpring=true
    bOverrideAngularDamping=true
    bUseGravityOverride=true
    gravityScale=0.35
    alphaBoolBlend {blendOption=Linear}
    boundBone {boneName=dyn_necklace_01}
    componentAppliedLinearAccClamp=<1400 1400 1400>
    componentLinearAccScale=<1 1 1>
    componentLinearVelScale=<1 1 1>
    gravityOverride {z=-980}
    physicsBodyDefinitions [{
      boundBone {boneName=dyn_necklace_01}
      boxExtents=<20 30 30>
      constraintSetup {angularLimitsMax=<15 2.5 15>, angularLimitsMin=<-15 -30 -15>, angularTarget=<1 0 0>}
      localJointOffset=<4 0 0>
     }]
    planarLimits [{drivingBone {boneName=spine_05}, planeTransform {location=<-0.1567 10.8941 0>, rotation=<rot 0 -17.2798 90>}}]
   }
   n33 = Anim Dynamics - Bone: dyn_necklace_02(ComponentPose=@n32, Alpha=1, ExternalForce="0,0,0") {
    angularDampingOverride=1
    angularSpringConstant=20000
    bAngularSpring=true
    bOverrideAngularDamping=true
    bUseGravityOverride=true
    gravityScale=0.35
    alphaBoolBlend {blendOption=Linear}
    boundBone {boneName=dyn_necklace_02}
    componentAppliedLinearAccClamp=<1400 1400 1400>
    componentLinearAccScale=<1 1 1>
    componentLinearVelScale=<1 1 1>
    gravityOverride {z=-980}
    physicsBodyDefinitions [{
      boundBone {boneName=dyn_necklace_02}
      boxExtents=<20 30 30>
      constraintSetup {angularLimitsMax=<15 5 15>, angularLimitsMin=<-15 -30 -15>, angularTarget=<1 0 0>}
      localJointOffset=<4 0 0>
     }]
    planarLimits [{drivingBone {boneName=spine_05}, planeTransform {location=<-0.2318 10.3203 0>, rotation=<rot 0 -6.0709 90.0001>}}]
   }
   n34 = "Transform (Modify) Bone - Bone: dyn_necklace_02"(ComponentPose=@n33, Alpha=1, Translation="0,0,-0.2", Rotation="0,0,0", Scale="1,1,0.7") {
    scaleMode=BMM_Additive
    scaleSpace=BCS_BoneSpace
    translationMode=BMM_Additive
    translationSpace=BCS_BoneSpace
    alphaBoolBlend {blendOption=Linear}
    boneToModify {boneName=dyn_necklace_02}
    scale {z=0.7}
    translation {z=-0.2}
   }
   n35 = Anim Dynamics - Bone: dyn_necklace_01(ComponentPose=Local To Component(LocalPose=Use cached pose 'Post Jacket Dynamics'()), Alpha=1, ExternalForce="0,0,0") {
    angularDampingOverride=0.975
    angularSpringConstant=20000
    bAngularSpring=true
    bOverrideAngularDamping=true
    bUseGravityOverride=true
    gravityScale=0.35
    alphaBoolBlend {blendOption=Linear}
    boundBone {boneName=dyn_necklace_01}
    componentAppliedLinearAccClamp=<2500 2500 2500>
    componentLinearAccScale=<0.33 0.33 0.33>
    componentLinearVelScale=<0.33 0.33 0.33>
    gravityOverride {z=-980}
    physicsBodyDefinitions [{
      boundBone {boneName=dyn_necklace_01}
      boxExtents=<20 30 30>
      constraintSetup {angularLimitsMax=<0 0 5>, angularLimitsMin=<0 -15 -5>, angularTarget=<1 0 0>}
      localJointOffset=<4 0 0>
     }]
    planarLimits [{drivingBone {boneName=spine_05}, planeTransform {location=<-0.1567 10.8941 0>, rotation=<rot 0 -17.2798 90.0001>}}]
    relativeSpaceBone {boneName=spine_05}
   }
   n36 = Anim Dynamics - Bone: dyn_necklace_02(ComponentPose=@n35, Alpha=1, ExternalForce="0,0,0") {
    angularDampingOverride=0.975
    angularSpringConstant=20000
    bAngularSpring=true
    bOverrideAngularDamping=true
    bUseGravityOverride=true
    gravityScale=0.35
    alphaBoolBlend {blendOption=Linear}
    boundBone {boneName=dyn_necklace_02}
    componentAppliedLinearAccClamp=<2500 2500 2500>
    componentLinearAccScale=<0.33 0.33 0.33>
    componentLinearVelScale=<0.33 0.33 0.33>
    gravityOverride {z=-980}
    physicsBodyDefinitions [{
      boundBone {boneName=dyn_necklace_02}
      boxExtents=<20 30 30>
      constraintSetup {angularLimitsMax=<0 0 5>, angularLimitsMin=<0 -15 -5>, angularTarget=<1 0 0>}
      localJointOffset=<4 0 0>
     }]
    planarLimits [{drivingBone {boneName=spine_05}, planeTransform {location=<-2.2273 11.5003 -3.3986>, rotation=<rot 0 -19.4217 90.0001>}}]
    relativeSpaceBone {boneName=spine_05}
   }
   Save cached pose 'Post Necklace Dynamics'(Pose=Blend Poses by bool(
     BlendPose_0=Component To Local(ComponentPose=@n34)
     BlendPose_1=Component To Local(ComponentPose=@n36)
     BlendTime_0=0
     BlendTime_1=0
     bActiveValue=Get bIsInFrontEnd()
    ) {blendType=Linear}) {cacheName=Post Necklace Dynamics}
  ]}
 EventGraph {nodes [
   n1 = Event BlueprintUpdateAnimation()
   > Set Delta Time X(Delta Time X=@n1.DeltaTimeX)
   > n2 = Sequence!(execute) -> (then_0, then_1, then_2:Exec, then_3:Exec)
   > n3 = Branch(Condition=Get bClothEnabled())
   > Cloth Controller Function DEPRECATED!(
    execute
    CharacterPartAnimInstance:Custom Character Part Anim Instance Object Reference=Self-Reference()
    ClothingAssetName:String=M_MED_LemonCartGranite_Clothing_0
    Cloth Paramaters:Cloth Params Struct Structure=Get Cloth Paramaters 0() -> (Cloth Paramaters 0:Cloth Params Struct Structure)
    Cloth Core Settings:Cloth Core Setting Struct Structure=Get Cloth Core Settings() -> (Cloth Core Settings:Cloth Core Setting Struct Structure)
    ClothUpAxis:Vector="0,0,1"
    OnGround_Moving_SpeedLimit:double=2
    OnGround_Moving_AngularVel_Max:double=0.05
    OnGround_Moving_FictitiousVel_Max:double=0.6
    LinVelScale_FwdBwd_Clamp:double=0
    LinVelScale_UpDwn_Clamp:double=0.7
    DeltaTime:double=Get Delta Time X()
    GameTimeInSeconds:double=GetGameTimeInSeconds()
    WindGustsParams:Wind Gust Params Struct Structure=Get Wind Gusts Params() -> (Wind Gusts Params:Wind Gust Params Struct Structure)
    WindGustLocalSettings:Wind Gust Local Setting Struct Structure=Get Wind Gust Local Settings() -> (Wind Gust Local Settings:Wind Gust Local Setting Struct Structure)
   ) -> (
    then
    CurrentStateOut:String
    Cloth Settings Out:Cloth Setting Struct Structure
    Cloth Core Settings Out:Cloth Core Setting Struct Structure
    WindGustSettings Out:Wind Gust Setting Struct Structure
    WindGustLocalSettings Out:Wind Gust Local Setting Struct Structure
    GustStrength - WindVelocityOut:Vector
    LinearVelVector-DELETE LATER:Vector
   )
   > SetClothingInteractorIfNotSet()
   > n4 = Sequence!(execute) -> (then_0, then_1, then_2)
   > Branch(Condition=Get Wearing Backbling())
   > SetAnimDrive!(
    execute
    self:Chaos Clothing Interactor Object Reference=Get ChaosClothingInteractor() -> (ChaosClothingInteractor:Chaos Clothing Interactor Object Reference)
    AnimDriveStiffness:Vector 2D Structure=(X=0.2,Y=1)
    AnimDriveDamping:Vector 2D Structure=(X=0.2,Y=1)
   ) -> (then)
   Branch(execute=@n4.then_1, Condition=Get bIsSkyDiving())
   > SetCollision!(
    execute
    self:Chaos Clothing Interactor Object Reference=Get ChaosClothingInteractor()
    CollisionThickness:float=3
    FrictionCoefficient:float=0.1
    bUseCCD:bool=false
    SelfCollisionThickness:float=0
   ) -> (then)
   n5 = Gravity Override State Function DEPRECATED(
    execute=@n3.else
    CharacterPartAnimInstance=Self-Reference()
    GameTimeInSeconds=GetGameTimeInSeconds()
    GravityOverrideParams=Get Gravity Override Params()
    RBAN Params=Get RBAN Params()
   )
   > Set Gravity Override(Gravity Override=@n5)
   > Set RBAN Settings Out(RBAN Settings Out:RBANSetting Struct Structure=@n5.RBAN SettingsOut) -> (Output_Get:RBANSetting Struct Structure)
   Set WPO Globals Function(
    execute=@n2.then_1
    Material=Get Mat Instance Dynamic()
    Overlay Material=Get Overlay Mat Instance Dynamic()
    Skeletal Mesh=Get Skel Mesh()
    "Local Noise Transform  Location"=Get Local Noise Transform Location() -> (Local Noise Transform Location:Vector)
    Local Noise Offset=Get Local Noise Offset()
   )
   > Set WPO Regional Function!(
    execute
    Material:Material Instance Dynamic Object Reference=Get Mat Instance Dynamic()
    Overlay Material:Material Instance Dynamic Object Reference=Get Overlay Mat Instance Dynamic()
    Skeletal Mesh:Skeletal Mesh Component Object Reference=Get Skel Mesh()
    Velocity Transform:Name=foot_r
    Velocity Interp Speed:double=Get Velocity Interp Speed()
    Collision Transform:Name=calf_r
    Collision Transform Axis:EAxis Enum
    Invert Collision Transform Axis:bool=false
    Delta Time X:double=Get Delta Time X()
    Previous Velocity Transform Location:"Vector (by ref)"=Get R1 Velocity Transform Location()
    Previous Velocity Transform Velocity:"Vector (by ref)"=Get R1 Velocity Transform Velocity()
    Region:int=1
   ) -> (then)
   > Set WPO Regional Function(
    Material=Get Mat Instance Dynamic()
    Overlay Material=Get Overlay Mat Instance Dynamic()
    Skeletal Mesh=Get Skel Mesh()
    Velocity Transform=foot_l
    Velocity Interp Speed=Get Velocity Interp Speed()
    Collision Transform=calf_l
    Invert Collision Transform Axis=true
    Delta Time X=Get Delta Time X()
    Previous Velocity Transform Location=Get R2 Velocity Transform Location()
    Previous Velocity Transform Velocity=Get R2 Velocity Transform Velocity()
    Region=2
   )
   > n6 = Branch(Condition=Get bIsSkydiveFloating())
   > n7 = Set WPO Pho Wind Intensity(WPO Pho Wind Intensity=0.65)
   SetScalarParameterValue(
    execute=[@n8.then, @n7.then]
    self=[Get Mat Instance Dynamic(), Get Overlay Mat Instance Dynamic()]
    ParameterName=ClothSim_WindStrength
    Value=Get WPO Pho Wind Intensity()
   )
   n8 = Set WPO Pho Wind Intensity(execute=@n6.else, WPO Pho Wind Intensity=0)
   n9 = Branch(
    execute=@n2.then_2
    Condition=Get bIsInFirstPerson!(self:Fort Player Anim Instance Object Reference=GetOwnerFortPlayerAnimInstance!(self:Custom Character Part Anim Instance Object Reference) -> (ReturnValue:Fort Player Anim Instance Object Reference)) -> (bIsInFirstPerson:bool)
   )
   n10 = GetPartSkeletalMeshComponent()
   > Set ClothBlendWeight!(execute, ClothBlendWeight:float=0, self:Skeletal Mesh Component Object Reference=@n10) -> (then, Output_Get:float)
   > Set bDisableClothSimulation!(execute, bDisableClothSimulation:bool=true, self:Skeletal Mesh Component Object Reference=@n10) -> (then, Output_Get:bool)
   Set ClothBlendWeight(execute=@n9.else, ClothBlendWeight=1, self=@n10)
   > Set bDisableClothSimulation(bDisableClothSimulation=false, self=@n10)
   Event OnCharacterCustomizationLoaded()
   > n11 = Sequence!(execute) -> (then_0, then_1, then_2, then_3:Exec, then_4:Exec)
   Set IsMaskOn(execute=@n11.then_1, IsMaskOn:bool=CharacterItemDefinitionHasAnyMetaTag(Tags=Make Array("[0]"=(TagName="Cosmetics.MetaTags.Style.OptionC")))) -> (Output_Get:bool)
   > Set IsScrapeVariant(IsScrapeVariant:bool=CharacterItemDefinitionHasAnyMetaTag(Tags=Make Array("[0]"=(TagName="Cosmetics.MetaTags.Style.OptionA")))) -> (Output_Get:bool)
   SetClothingInteractorIfNotSet(execute=@n11.then_4)
   n12 = Set Skel Mesh(execute=@n11.then_0, Skel Mesh=GetPartSkeletalMeshComponent())
   > n13 = Get Dynamic Material Instance(SkeletalMeshComponent=@n12)
   > Set Mat Instance Dynamic(Mat Instance Dynamic=@n13)
   > n14 = GetOrCreateDynamicOverlayMaterial(MeshComponent=Get Skel Mesh(), MaterialName=None)
   > Set Overlay Mat Instance Dynamic(Overlay Mat Instance Dynamic=@n14)
   n15 = CheckForBackbling()
   n16 = OR Boolean(A=@n15, B=@n15.bIsBackblingLong)
   Set Wearing Backbling(execute=@n11.then_3, Wearing Backbling:bool=@n16) -> (Output_Get:bool)
   > Set Is Backbling A Cape(Is Backbling A Cape:bool=@n15.bIsBackblingACape) -> (Output_Get:bool)
   n17 = "To Float (Boolean)"(InBool=@n16)
   > Set Wearing Backbling Float(Wearing Backbling Float:double=@n17) -> (Output_Get:double)
   > Set Not Wearing Backbling Inverse Float(Not Wearing Backbling Inverse Float:double=float - float(A=1, B=@n17)) -> (Output_Get:double)
   Branch(Condition=Get bIsPlayingMeleeAnim())
   n18 = Get ChaosClothingInteractor()
   > SetAnimDrive(self=@n18, AnimDriveStiffness=(X=0.1,Y=1), AnimDriveDamping=(X=0.1,Y=1))
   > SetMaterial!(
    execute
    self:Chaos Clothing Interactor Object Reference=@n18
    EdgeStiffness:Vector 2D Structure=(X=1,Y=1)
    BendingStiffness:Vector 2D Structure=(X=1,Y=1)
    AreaStiffness:Vector 2D Structure=(X=1,Y=1)
   ) -> (then)
  ]}
 SetClothingInteractorIfNotSet {nodes [
   SetClothingInteractorIfNotSet() {is=title K2Node_FunctionEntry}
   n1 = IsValid(Object=Get ChaosClothingInteractor())
   > Branch(Condition=AND Boolean(A=Get bClothEnabled(), B=NOT Boolean(A=@n1)))
   n2 = MakeLiteralString(Value=M_MED_LemonCartGranite_Clothing_0)
   n3 = GetOwningComponent()
   > n4 = Cast To ChaosClothingInteractor!(
    execute
    Object=GetClothingInteractor!(
     self:Clothing Simulation Interactor Object Reference=GetClothingSimulationInteractor!(self:Skeletal Mesh Component Object Reference=@n3, ClothingAssetName:Name=None) -> (ReturnValue:Clothing Simulation Interactor Object Reference)
     ClothingAssetName:Name=String To Name(InString=@n2)
    ) -> (ReturnValue:Clothing Interactor Object Reference)
   ) -> (then, CastFailed, AsChaos Clothing Interactor:Chaos Clothing Interactor Object Reference) {is=title K2Node_DynamicCast}
   > Set ChaosClothingInteractor(ChaosClothingInteractor:Chaos Clothing Interactor Object Reference=@n4) -> (Output_Get:Chaos Clothing Interactor Object Reference)
  ]}
]
variables [
 ChaosClothingInteractor {type=ChaosClothingInteractor, default=null}
 Cloth Core Settings {
  type=ClothCoreSettingStruct
  default {
   collisionThickness=1
   dampingCoefficient=0.05
   frictionCoefficient=0.5
   localDampingCoefficient=0.1
   numberIterations=2
   numberSubsteps=1
   selfCollisionThickness=0
   animDriveDamping=<0 1>
   animDriveStiffness=<0.05 1>
   areaStiffness=<0 1>
   bendingStiffness=<0.9 1>
   edgeStiffness=<0.9 1>
   tetherStiffness=<0 1>
  }
 }
 Cloth Paramaters 0 {
  type=ClothParamsStruct
  default {
   dBNO {
    angularVelocityScale=0.6
    fictitiousAngularScale=0.1
    jointName=pelvis
    dragCoefficient=<0.2 1>
    gravityOverride=<-980 400 0>
    liftCoefficient=<0 0.5>
    linearVelocityScale=<0.3 0.3 0.1>
   }
   dBNO_Carried {
    angularVelocityScale=0.6
    fictitiousAngularScale=0.1
    jointName=pelvis
    dragCoefficient=<0.2 1>
    gravityOverride=<-980 -200 0>
    liftCoefficient=<0 0.5>
    linearVelocityScale=<0.3 0.3 0.1>
   }
   dBNO_Impulse {
    angularVelocityScale=0
    fictitiousAngularScale=0
    jointName=pelvis
    dragCoefficient=<0 0.2>
    gravityOverride=<-3000 -2000 0>
    liftCoefficient=<0 0.1>
    linearVelocityScale=<0.3 0.1 0.3>
   }
   dBNO_RollBwd {
    angularVelocityScale=0
    fictitiousAngularScale=0
    jointName=pelvis
    dragCoefficient=<0 0.2>
    gravityOverride=<-2000 0 0>
    liftCoefficient=<0 0.1>
    linearVelocityScale=<0.3 0.1 0.3>
   }
   dBNO_RollFwd {
    angularVelocityScale=0
    fictitiousAngularScale=0
    jointName=pelvis
    dragCoefficient=<0 0.2>
    gravityOverride=<-980 0 0>
    liftCoefficient=<0 0.1>
    linearVelocityScale=<0.3 0.1 0.3>
   }
   dBNO_Tumbling {
    angularVelocityScale=0
    fictitiousAngularScale=0
    jointName=pelvis
    dragCoefficient=<0 0.2>
    gravityOverride=<-2500 -300 0>
    liftCoefficient=<0 0.1>
    linearVelocityScale=<0.3 0.1 0.3>
   }
   emoteOrMelee {
    angularVelocityScale=0.5
    fictitiousAngularScale=0.3
    jointName=spine_05
    dragCoefficient=<0 1>
    gravityOverride=<-980 0 0>
    liftCoefficient=<0 1>
    linearVelocityScale=<0.5 0.5 0.5>
   }
   front_End {
    angularVelocityScale=0.1
    fictitiousAngularScale=0.2
    jointName=root
    dragCoefficient=<0 0>
    gravityOverride=<0 0 -980>
    liftCoefficient=<0 0>
    linearVelocityScale=<0.1 0.1 0.1>
   }
   onGround_Crouch_Moving {
    angularVelocityScale=0.1
    fictitiousAngularScale=0.3
    jointName=root
    dragCoefficient=<0 0>
    gravityOverride=<0 0 -980>
    liftCoefficient=<0 0>
    linearVelocityScale=<0.2 0.2 0.2>
   }
   onGround_Falling {
    angularVelocityScale=0.1
    fictitiousAngularScale=0
    jointName=pelvis
    dragCoefficient=<0.1 1>
    gravityOverride=<200 -700 0>
    liftCoefficient=<0.1 1>
    linearVelocityScale=<0.2 0.2 0.15>
   }
   onGround_Moving {
    angularVelocityScale=0.3
    fictitiousAngularScale=0.1
    jointName=root
    dragCoefficient=<0 0.2>
    gravityOverride=<0 -100 -980>
    liftCoefficient=<0 0.2>
    linearVelocityScale=<0.3 0.3 0.3>
   }
   onGround_Standing {
    angularVelocityScale=0.05
    fictitiousAngularScale=0.6
    jointName=root
    dragCoefficient=<0 0.5>
    gravityOverride=<0 0 -980>
    liftCoefficient=<0 0.5>
    linearVelocityScale=<0.4 0.4 0.4>
   }
   rail_Grinding {
    angularVelocityScale=0.2
    fictitiousAngularScale=0
    jointName=root
    dragCoefficient=<0.5 2>
    gravityOverride=<50 -500 -400>
    liftCoefficient=<0.5 1>
    linearVelocityScale=<0.1 0.1 0.1>
   }
   riding_Moving {
    angularVelocityScale=0.05
    fictitiousAngularScale=1
    jointName=root
    dragCoefficient=<0.5 0.5>
    gravityOverride=<0 0 -980>
    liftCoefficient=<0.5 0.5>
    linearVelocityScale=<0.4 0.6 0.3>
   }
   riding_Moving_Targeting {
    angularVelocityScale=0.05
    fictitiousAngularScale=0.7
    jointName=root
    dragCoefficient=<0.5 0.5>
    gravityOverride=<0 0 -980>
    liftCoefficient=<0.5 0.5>
    linearVelocityScale=<0.2 0.6 0.2>
   }
   riding_Standing {
    angularVelocityScale=0.05
    fictitiousAngularScale=1
    jointName=root
    dragCoefficient=<0.5 0.5>
    gravityOverride=<0 -450 -980>
    liftCoefficient=<0.5 0.5>
    linearVelocityScale=<0.6 0.6 0.3>
   }
   skydive_DivingDown {
    angularVelocityScale=0.05
    fictitiousAngularScale=0
    jointName=pelvis
    dragCoefficient=<0 0.3>
    gravityOverride=<-200 0 0>
    liftCoefficient=<0 0.3>
    linearVelocityScale=<0.05 0.05 0.05>
   }
   skydive_FloatingDown {
    angularVelocityScale=0.1
    fictitiousAngularScale=0
    jointName=root
    dragCoefficient=<0 0.8>
    gravityOverride=<0 -1000 -700>
    liftCoefficient=<0 1.5>
    linearVelocityScale=<0.1 0.1 0.1>
   }
   skydive_GoingUp {
    angularVelocityScale=0.1
    fictitiousAngularScale=0
    jointName=root
    dragCoefficient=<0 2>
    gravityOverride=<0 -2000 -2000>
    liftCoefficient=<0 2>
    linearVelocityScale=<0.1 0.1 0.05>
   }
   skydive_Parachuting {
    angularVelocityScale=0.1
    fictitiousAngularScale=0
    jointName=root
    dragCoefficient=<0.1 1>
    gravityOverride=<0 -200 -980>
    liftCoefficient=<0 1>
    linearVelocityScale=<0.1 0.1 0.1>
   }
   skydive_Surfing {
    angularVelocityScale=0.6
    fictitiousAngularScale=0.1
    jointName=root
    dragCoefficient=<0 0.5>
    gravityOverride=<0 -600 -500>
    liftCoefficient=<0 0.5>
    linearVelocityScale=<0.2 0.2 0.2>
   }
   swimming {
    angularVelocityScale=0.2
    fictitiousAngularScale=0
    jointName=root
    dragCoefficient=<0 1>
    gravityOverride=<0 -600 -200>
    liftCoefficient=<0 0.2>
    linearVelocityScale=<0.2 0.2 0.2>
   }
   tactical_Sprint {
    angularVelocityScale=0.3
    fictitiousAngularScale=0.1
    jointName=root
    dragCoefficient=<1 1>
    gravityOverride=<0 0 -980>
    liftCoefficient=<0.5 0.5>
    linearVelocityScale=<0.3 0.3 0.5>
   }
   ziplining {
    angularVelocityScale=0.2
    fictitiousAngularScale=0
    jointName=root
    dragCoefficient=<0.2 0.5>
    gravityOverride=<0 -50 -980>
    liftCoefficient=<0.2 0.5>
    linearVelocityScale=<0.05 0.05 0.05>
   }
  }
 }
 Component Linear Acc Scale {type=Vector, default=<0.3 0.3 0.3>}
 Component Linear Acc Scale BACKPACK {type=Vector}
 Component Linear Acc Scale DEFAULT {type=Vector, default=<0.3 0.3 0.3>}
 Component Linear Vel Scale {type=Vector, default=<0.3 0.3 0.3>}
 Component Linear Vel Scale BACKPACK {type=Vector}
 Component Linear Vel Scale DEFAULT {type=Vector, default=<0.3 0.3 0.3>}
 Delta Time X {type=number}
 Gravity Override {type=Vector}
 Gravity Override Params {
  type=GravityOverrideParamsStruct
  default {
   rail_Grinding {windAmplitude=<200 -500 900>, windFrequency=<1.5 7 20>, windOffset=<0 -600 400>}
   ziplining {jointName=pelvis, gravityOverride=<-980 0 0>, windAmplitude=<1200 -500 200>, windFrequency=<16 7 0>, windOffset=<400 -600 0>}
  }
 }
 Is Backbling A Cape {type=boolean}
 IsBaseVariant {type=boolean}
 IsMaskOn {type=boolean}
 IsScrapeVariant {type=boolean}
 Local Noise Offset {type=Vector}
 Local Noise Transform Location {type=Vector}
 Mat Instance Dynamic {type=MaterialInstanceDynamic}
 Not Wearing Backbling Inverse Float {type=number}
 Overlay Mat Instance Dynamic {type=MaterialInstanceDynamic}
 R1 Velocity Transform Location {type=Vector}
 R1 Velocity Transform Velocity {type=Vector}
 R2 Velocity Transform Location {type=Vector}
 R2 Velocity Transform Velocity {type=Vector}
 RBAN Params {
  type=RBANParamsStruct
  default {
   dBNO {
    componentLinearAccScale=<0.15 0.15 0.15>
    componentLinearVelScale=<0.15 0.15 0.15>
    simSpaceSettings {maxAngularAcceleration=5, maxAngularVelocity=5, maxLinearAcceleration=1, maxLinearVelocity=1, velocityScaleZ=1, externalLinearDragV=<0 0 0>}
   }
   dBNO_Carried {
    componentLinearAccScale=<0.15 0.15 0.15>
    componentLinearVelScale=<0.15 0.15 0.15>
    simSpaceSettings {maxAngularAcceleration=5, maxAngularVelocity=5, maxLinearAcceleration=1, maxLinearVelocity=1, velocityScaleZ=1, externalLinearDragV=<0 0 0>}
   }
   emoteOrMelee {
    componentLinearAccScale=<0.15 0.15 0.15>
    componentLinearVelScale=<0.15 0.15 0.15>
    simSpaceSettings {maxAngularAcceleration=5, maxAngularVelocity=5, maxLinearAcceleration=1, maxLinearVelocity=1, velocityScaleZ=1, externalLinearDragV=<0 0 0>}
   }
   front_End {simSpaceSettings {
     maxAngularAcceleration=5
     maxAngularVelocity=5
     maxLinearAcceleration=1
     maxLinearVelocity=1
     velocityScaleZ=1
     worldAlpha=0.5
     externalLinearDragV=<0 0 0>
    }}
   onGround_Crouch_Moving {
    componentLinearAccScale=<0.15 0.15 0.15>
    componentLinearVelScale=<0.15 0.15 0.15>
    simSpaceSettings {maxAngularAcceleration=5, maxAngularVelocity=5, maxLinearAcceleration=1, maxLinearVelocity=1, velocityScaleZ=1, externalLinearDragV=<0 0 0>}
   }
   onGround_Falling {
    componentLinearAccScale=<0.15 0.15 0.15>
    componentLinearVelScale=<0.15 0.15 0.15>
    simSpaceSettings {maxAngularAcceleration=5, maxAngularVelocity=5, maxLinearAcceleration=1, maxLinearVelocity=1, velocityScaleZ=1, externalLinearDragV=<0 0 0>}
   }
   onGround_Moving {
    componentLinearAccScale=<0.15 0.15 0.15>
    componentLinearVelScale=<0.15 0.15 0.15>
    simSpaceSettings {
     maxAngularAcceleration=10
     maxAngularVelocity=10
     maxLinearAcceleration=1
     maxLinearVelocity=1
     velocityScaleZ=1
     externalLinearDragV=<0 0 0>
    }
   }
   onGround_Standing {
    componentLinearAccScale=<0.15 0.15 0.15>
    componentLinearVelScale=<0.15 0.15 0.15>
    simSpaceSettings {maxAngularAcceleration=5, maxAngularVelocity=5, maxLinearAcceleration=1, maxLinearVelocity=1, velocityScaleZ=1, externalLinearDragV=<0 0 0>}
   }
   rail_Grinding {
    componentLinearAccScale=<0.15 0.15 0.15>
    componentLinearVelScale=<0.15 0.15 0.15>
    simSpaceSettings {
     maxAngularAcceleration=5
     maxAngularVelocity=5
     maxLinearAcceleration=1
     maxLinearVelocity=1
     velocityScaleZ=1
     worldAlpha=1
     externalLinearDragV=<0 0 0>
    }
   }
   riding_Moving {
    componentLinearAccScale=<0.15 0.15 0.15>
    componentLinearVelScale=<0.15 0.15 0.15>
    simSpaceSettings {maxAngularAcceleration=5, maxAngularVelocity=5, maxLinearAcceleration=1, maxLinearVelocity=1, velocityScaleZ=1, externalLinearDragV=<0 0 0>}
   }
   riding_Moving_Targeting {
    componentLinearAccScale=<0.15 0.15 0.15>
    componentLinearVelScale=<0.15 0.15 0.15>
    simSpaceSettings {
     maxAngularAcceleration=1000
     maxAngularVelocity=3
     maxLinearAcceleration=300
     maxLinearVelocity=700
     velocityScaleZ=0
     externalLinearDragV=<0 0.5 0>
    }
   }
   riding_Standing {
    componentLinearAccScale=<0.15 0.15 0.15>
    componentLinearVelScale=<0.15 0.15 0.15>
    simSpaceSettings {maxAngularAcceleration=5, maxAngularVelocity=5, maxLinearAcceleration=1, maxLinearVelocity=1, velocityScaleZ=1, externalLinearDragV=<0 0 0>}
   }
   skydive_DivingDown {
    componentLinearAccScale=<0.15 0.15 0.15>
    simSpaceSettings {maxAngularAcceleration=5, maxAngularVelocity=5, maxLinearAcceleration=1, maxLinearVelocity=1, velocityScaleZ=1, externalLinearDragV=<0 0 0>}
   }
   skydive_FloatingDown {
    componentLinearAccScale=<0.15 0.15 0.15>
    componentLinearVelScale=<0.15 0.15 0.15>
    simSpaceSettings {maxAngularAcceleration=5, maxAngularVelocity=5, maxLinearAcceleration=1, maxLinearVelocity=1, velocityScaleZ=1, externalLinearDragV=<0 0 0>}
   }
   skydive_GoingUp {
    componentLinearAccScale=<0.15 0.15 0.15>
    componentLinearVelScale=<0.15 0.15 0.15>
    simSpaceSettings {maxAngularAcceleration=5, maxAngularVelocity=5, maxLinearAcceleration=1, maxLinearVelocity=1, velocityScaleZ=1, externalLinearDragV=<0 0 0>}
   }
   skydive_Parachuting {
    componentLinearAccScale=<0.15 0.15 0.15>
    componentLinearVelScale=<0.15 0.15 0.15>
    simSpaceSettings {maxAngularAcceleration=5, maxAngularVelocity=5, maxLinearAcceleration=1, maxLinearVelocity=1, velocityScaleZ=1, externalLinearDragV=<0 0 0>}
   }
   skydive_Surfing {
    componentLinearAccScale=<0.15 0.15 0.15>
    simSpaceSettings {maxAngularAcceleration=5, maxAngularVelocity=5, maxLinearAcceleration=1, maxLinearVelocity=1, velocityScaleZ=1, externalLinearDragV=<0 0 0>}
   }
   swimming {
    componentLinearAccScale=<0.15 0.15 0.15>
    componentLinearVelScale=<0.15 0.15 0.15>
    simSpaceSettings {maxAngularAcceleration=5, maxAngularVelocity=5, maxLinearAcceleration=1, maxLinearVelocity=1, velocityScaleZ=1, externalLinearDragV=<0 0 0>}
   }
   tactical_Sprint {
    componentLinearAccScale=<0.15 0.15 0.15>
    componentLinearVelScale=<0.15 0.15 0.15>
    simSpaceSettings {maxAngularAcceleration=5, maxAngularVelocity=5, maxLinearAcceleration=1, maxLinearVelocity=1, velocityScaleZ=1, externalLinearDragV=<0 0 0>}
   }
   ziplining {simSpaceSettings {
     maxAngularAcceleration=5
     maxAngularVelocity=5
     maxLinearAcceleration=1
     maxLinearVelocity=1
     velocityScaleZ=1
     worldAlpha=0
     externalLinearDragV=<0 0 0>
    }}
  }
 }
 RBAN Settings Out {
  type=RBANSettingStruct
  default {
   componentLinearAccScale=<0 0 0>
   componentLinearVelScale=<0 0 0>
   simSpaceSettings {
    dampingAlpha=1
    maxAngularAcceleration=1000
    maxAngularVelocity=3
    maxLinearAcceleration=300
    maxLinearVelocity=700
    velocityScaleZ=0
    worldAlpha=1
    externalAngularVelocity=<0 0 0>
    externalLinearDragV=<0 0.5 0>
    externalLinearVelocity=<0 0 0>
   }
  }
 }
 Sim Space Settings {
  type=SimSpaceSettings
  default {
   dampingAlpha=1
   maxAngularAcceleration=5
   maxAngularVelocity=5
   maxLinearAcceleration=1
   maxLinearVelocity=1
   velocityScaleZ=1
   worldAlpha=1
   externalLinearDragV=<0 0 0>
  }
 }
 Sim Space Settings BACKPACK {
  type=SimSpaceSettings
  default {
   dampingAlpha=1
   maxAngularAcceleration=2.5
   maxAngularVelocity=2.5
   maxLinearAcceleration=0.5
   maxLinearVelocity=0.5
   velocityScaleZ=0.5
   worldAlpha=0.5
   externalLinearDragV=<0 0 0>
  }
 }
 Sim Space Settings DEFAULT {
  type=SimSpaceSettings
  default {
   dampingAlpha=1
   maxAngularAcceleration=5
   maxAngularVelocity=5
   maxLinearAcceleration=1
   maxLinearVelocity=1
   velocityScaleZ=1
   worldAlpha=1
   externalLinearDragV=<0 0 0>
  }
 }
 Skel Mesh {type=SkeletalMeshComponent}
 Velocity Interp Speed {type=number}
 Wearing Backbling {type=boolean}
 Wearing Backbling Float {type=number}
 Wind Gust Local Settings {type=WindGustLocalSettingStruct, default {elapsedTime=0, finalGustStrengthLocal=0, gustDuration=0, gustStrength=0, timeBetweenGusts=0}}
 Wind Gusts Params {
  type=WindGustParamsStruct
  default {
   dBNO {
    maxDownInterpSpeed=5
    maxGustDuration=2
    maxGustStrength=7
    maxTimeBetweenGusts=7
    maxUpInterpSpeed=5
    minDownInterpSpeed=2
    minGustDuration=0.5
    minGustStrength=3
    minTimeBetweenGusts=0
    minUpInterpSpeed=2
    windTurbulenceScalarA=0.1
    windTurbulenceScalarB=1
    windAmplitude=<0 0 0>
    windFrequency=<0 0 0>
    windOffset=<0 0 0>
   }
   dBNO_Carried {
    maxDownInterpSpeed=5
    maxGustDuration=2
    maxGustStrength=7
    maxTimeBetweenGusts=7
    maxUpInterpSpeed=5
    minDownInterpSpeed=2
    minGustDuration=0.5
    minGustStrength=3
    minTimeBetweenGusts=0
    minUpInterpSpeed=2
    windTurbulenceScalarA=0.1
    windTurbulenceScalarB=1
    windAmplitude=<0 0 0>
    windFrequency=<0 0 0>
    windOffset=<0 0 0>
   }
   dBNO_Impulse {
    maxDownInterpSpeed=5
    maxGustDuration=2
    maxGustStrength=7
    maxTimeBetweenGusts=7
    maxUpInterpSpeed=5
    minDownInterpSpeed=2
    minGustDuration=0.5
    minGustStrength=3
    minTimeBetweenGusts=0
    minUpInterpSpeed=2
    windTurbulenceScalarA=0.1
    windTurbulenceScalarB=1
    windAmplitude=<0 0 0>
    windFrequency=<0 0 0>
    windOffset=<0 0 0>
   }
   dBNO_RollBwd {
    maxDownInterpSpeed=5
    maxGustDuration=2
    maxGustStrength=7
    maxTimeBetweenGusts=7
    maxUpInterpSpeed=5
    minDownInterpSpeed=2
    minGustDuration=0.5
    minGustStrength=3
    minTimeBetweenGusts=0
    minUpInterpSpeed=2
    windTurbulenceScalarA=0.1
    windTurbulenceScalarB=1
    windAmplitude=<0 0 0>
    windFrequency=<0 0 0>
    windOffset=<0 0 0>
   }
   dBNO_RollFwd {
    maxDownInterpSpeed=5
    maxGustDuration=2
    maxGustStrength=7
    maxTimeBetweenGusts=7
    maxUpInterpSpeed=5
    minDownInterpSpeed=2
    minGustDuration=0.5
    minGustStrength=3
    minTimeBetweenGusts=0
    minUpInterpSpeed=2
    windTurbulenceScalarA=0.1
    windTurbulenceScalarB=1
    windAmplitude=<0 0 0>
    windFrequency=<0 0 0>
    windOffset=<0 0 0>
   }
   dBNO_Tumbling {
    maxDownInterpSpeed=5
    maxGustDuration=2
    maxGustStrength=7
    maxTimeBetweenGusts=7
    maxUpInterpSpeed=5
    minDownInterpSpeed=2
    minGustDuration=0.5
    minGustStrength=3
    minTimeBetweenGusts=0
    minUpInterpSpeed=2
    windTurbulenceScalarA=0.1
    windTurbulenceScalarB=1
    windAmplitude=<0 0 0>
    windFrequency=<0 0 0>
    windOffset=<0 0 0>
   }
   emoteOrMelee {
    maxDownInterpSpeed=5
    maxGustDuration=2
    maxGustStrength=7
    maxTimeBetweenGusts=7
    maxUpInterpSpeed=5
    minDownInterpSpeed=2
    minGustDuration=0.5
    minGustStrength=3
    minTimeBetweenGusts=0
    minUpInterpSpeed=2
    windTurbulenceScalarA=0.1
    windTurbulenceScalarB=1
    windAmplitude=<0 0 0>
    windFrequency=<0 0 0>
    windOffset=<0 0 0>
   }
   front_End {
    maxDownInterpSpeed=5
    maxGustDuration=2
    maxGustStrength=4
    maxTimeBetweenGusts=7
    maxUpInterpSpeed=5
    minDownInterpSpeed=2
    minGustDuration=0.5
    minGustStrength=2
    minTimeBetweenGusts=0
    minUpInterpSpeed=2
    windTurbulenceScalarA=0.1
    windTurbulenceScalarB=1
    windAmplitude=<10 30 50>
    windFrequency=<10 3 4>
    windOffset=<0 -200 -250>
   }
   onGround_Crouch_Moving {
    maxDownInterpSpeed=5
    maxGustDuration=2
    maxGustStrength=7
    maxTimeBetweenGusts=7
    maxUpInterpSpeed=5
    minDownInterpSpeed=2
    minGustDuration=0.5
    minGustStrength=3
    minTimeBetweenGusts=0
    minUpInterpSpeed=2
    windTurbulenceScalarA=0.1
    windTurbulenceScalarB=1
    windAmplitude=<100 25 75>
    windFrequency=<1.5 0.5 7.5>
    windOffset=<0 -100 -175>
   }
   onGround_Falling {
    maxDownInterpSpeed=5
    maxGustDuration=2
    maxGustStrength=7
    maxTimeBetweenGusts=7
    maxUpInterpSpeed=5
    minDownInterpSpeed=2
    minGustDuration=0.5
    minGustStrength=3
    minTimeBetweenGusts=0
    minUpInterpSpeed=2
    windTurbulenceScalarA=0.1
    windTurbulenceScalarB=1
    windAmplitude=<0 0 0>
    windFrequency=<0 0 0>
    windOffset=<0 0 0>
   }
   onGround_Moving {
    maxDownInterpSpeed=5
    maxGustDuration=8
    maxGustStrength=7
    maxTimeBetweenGusts=8
    maxUpInterpSpeed=5
    minDownInterpSpeed=2
    minGustDuration=0.2
    minGustStrength=3
    minTimeBetweenGusts=1
    minUpInterpSpeed=2
    windTurbulenceScalarA=0.1
    windTurbulenceScalarB=0.8
    windAmplitude=<30 65 42>
    windFrequency=<5 6 8.6>
    windOffset=<30 -250 -300>
   }
   onGround_Standing {
    maxDownInterpSpeed=5
    maxGustDuration=8
    maxGustStrength=7
    maxTimeBetweenGusts=8
    maxUpInterpSpeed=5
    minDownInterpSpeed=2
    minGustDuration=0.2
    minGustStrength=2
    minTimeBetweenGusts=1
    minUpInterpSpeed=2
    windTurbulenceScalarA=0.1
    windTurbulenceScalarB=0.8
    windAmplitude=<0 0 0>
    windFrequency=<0 0 0>
    windOffset=<0 0 0>
   }
   rail_Grinding {
    maxDownInterpSpeed=5
    maxGustDuration=8
    maxGustStrength=7
    maxTimeBetweenGusts=8
    maxUpInterpSpeed=5
    minDownInterpSpeed=2
    minGustDuration=0.2
    minGustStrength=3
    minTimeBetweenGusts=1
    minUpInterpSpeed=2
    windTurbulenceScalarA=0.1
    windTurbulenceScalarB=0.8
    windAmplitude=<20 65 42>
    windFrequency=<5 6 8.6>
    windOffset=<50 -500 -400>
   }
   riding_Moving {
    maxDownInterpSpeed=5
    maxGustDuration=2
    maxGustStrength=7
    maxTimeBetweenGusts=7
    maxUpInterpSpeed=5
    minDownInterpSpeed=2
    minGustDuration=0.5
    minGustStrength=3
    minTimeBetweenGusts=0
    minUpInterpSpeed=2
    windTurbulenceScalarA=0.1
    windTurbulenceScalarB=1
    windAmplitude=<0 0 0>
    windFrequency=<0 0 0>
    windOffset=<0 0 0>
   }
   riding_Moving_Targeting {
    maxDownInterpSpeed=5
    maxGustDuration=2
    maxGustStrength=7
    maxTimeBetweenGusts=7
    maxUpInterpSpeed=5
    minDownInterpSpeed=2
    minGustDuration=0.5
    minGustStrength=3
    minTimeBetweenGusts=0
    minUpInterpSpeed=2
    windTurbulenceScalarA=0.1
    windTurbulenceScalarB=1
    windAmplitude=<0 0 0>
    windFrequency=<0 0 0>
    windOffset=<0 0 0>
   }
   riding_Standing {
    maxDownInterpSpeed=5
    maxGustDuration=2
    maxGustStrength=7
    maxTimeBetweenGusts=7
    maxUpInterpSpeed=5
    minDownInterpSpeed=2
    minGustDuration=0.5
    minGustStrength=3
    minTimeBetweenGusts=0
    minUpInterpSpeed=2
    windTurbulenceScalarA=0.1
    windTurbulenceScalarB=1
    windAmplitude=<0 0 0>
    windFrequency=<0 0 0>
    windOffset=<0 0 0>
   }
   skydive_DivingDown {
    maxDownInterpSpeed=5
    maxGustDuration=8
    maxGustStrength=7
    maxTimeBetweenGusts=8
    maxUpInterpSpeed=5
    minDownInterpSpeed=2
    minGustDuration=0.2
    minGustStrength=3
    minTimeBetweenGusts=1
    minUpInterpSpeed=2
    windTurbulenceScalarA=0.1
    windTurbulenceScalarB=4
    windAmplitude=<0 15 25>
    windFrequency=<0 4 10>
    windOffset=<-450 -100 100>
   }
   skydive_FloatingDown {
    maxDownInterpSpeed=5
    maxGustDuration=8
    maxGustStrength=1
    maxTimeBetweenGusts=8
    maxUpInterpSpeed=5
    minDownInterpSpeed=2
    minGustDuration=0.2
    minGustStrength=0
    minTimeBetweenGusts=1
    minUpInterpSpeed=2
    windTurbulenceScalarA=0.5
    windTurbulenceScalarB=1
    windAmplitude=<50 20 35>
    windFrequency=<8 2 3>
    windOffset=<0 -300 200>
   }
   skydive_GoingUp {
    maxDownInterpSpeed=5
    maxGustDuration=8
    maxGustStrength=7
    maxTimeBetweenGusts=8
    maxUpInterpSpeed=5
    minDownInterpSpeed=2
    minGustDuration=0.2
    minGustStrength=3
    minTimeBetweenGusts=1
    minUpInterpSpeed=2
    windTurbulenceScalarA=0.1
    windTurbulenceScalarB=0.8
    windAmplitude=<0 0 0>
    windFrequency=<0 0 0>
    windOffset=<0 -500 300>
   }
   skydive_Parachuting {
    maxDownInterpSpeed=5
    maxGustDuration=8
    maxGustStrength=1
    maxTimeBetweenGusts=8
    maxUpInterpSpeed=5
    minDownInterpSpeed=2
    minGustDuration=5
    minGustStrength=0
    minTimeBetweenGusts=5
    minUpInterpSpeed=2
    windTurbulenceScalarA=0.2
    windTurbulenceScalarB=1
    windAmplitude=<25 50 75>
    windFrequency=<8 6 2>
    windOffset=<-150 -400 -150>
   }
   skydive_Surfing {
    maxDownInterpSpeed=5
    maxGustDuration=8
    maxGustStrength=1
    maxTimeBetweenGusts=8
    maxUpInterpSpeed=5
    minDownInterpSpeed=2
    minGustDuration=5
    minGustStrength=0
    minTimeBetweenGusts=5
    minUpInterpSpeed=2
    windTurbulenceScalarA=0.2
    windTurbulenceScalarB=1
    windAmplitude=<25 50 75>
    windFrequency=<8 6 2>
    windOffset=<-150 -400 -150>
   }
   swimming {
    maxDownInterpSpeed=5
    maxGustDuration=2
    maxGustStrength=7
    maxTimeBetweenGusts=7
    maxUpInterpSpeed=5
    minDownInterpSpeed=2
    minGustDuration=0.5
    minGustStrength=3
    minTimeBetweenGusts=0
    minUpInterpSpeed=2
    windTurbulenceScalarA=0.1
    windTurbulenceScalarB=1
    windAmplitude=<10 25 20>
    windFrequency=<1.5 0.5 7.5>
    windOffset=<0 -150 50>
   }
   tactical_Sprint {
    maxDownInterpSpeed=5
    maxGustDuration=2
    maxGustStrength=7
    maxTimeBetweenGusts=7
    maxUpInterpSpeed=5
    minDownInterpSpeed=2
    minGustDuration=0.5
    minGustStrength=3
    minTimeBetweenGusts=0
    minUpInterpSpeed=2
    windTurbulenceScalarA=0.1
    windTurbulenceScalarB=1
    windAmplitude=<0 0 0>
    windFrequency=<0 0 0>
    windOffset=<0 0 0>
   }
   ziplining {
    maxDownInterpSpeed=5
    maxGustDuration=8
    maxGustStrength=7
    maxTimeBetweenGusts=8
    maxUpInterpSpeed=5
    minDownInterpSpeed=2
    minGustDuration=0.2
    minGustStrength=3
    minTimeBetweenGusts=1
    minUpInterpSpeed=2
    windTurbulenceScalarA=0.1
    windTurbulenceScalarB=0.8
    windAmplitude=<20 65 42>
    windFrequency=<5 6 8.6>
    windOffset=<50 -500 -400>
   }
  }
 }
 WPO Pho Wind Intensity {type=number}
]
