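# Physics asset supplying cloth collision capsules for the LemonCartGranite medium male player body, apparently for its jacket.
# Capsules cover the pelvis and three spine bones, with paired capsules on pelvis, spine_02 and spine_03.
# Two dedicated jacket-collision helper bones add larger, offset, rotated capsules, constrained to pelvis and spine_01.
# Thigh, calf and foot bodies carry no shapes, and all constraints leave swing and twist free.
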PhysicsAsset /BRCosmetics/Characters/Player/Male/Medium/Bodies/M_MED_LemonCartGranite/Meshes/M_MED_LemonCartGranite_ClothColliders
bodies {
 calf_l []
 calf_r []
 dyn_cloth_jacket_collision_spine_01 [{
   shapeType=Capsule
   length=28.75
   radius=21.1875
   shapeName=dyn_cloth_jacket_collision_spine_01_capsule
   center=<-2.6271 -37.9203 0>
   rotation=<rot 0 -80.7505 0>
  }]
 dyn_cloth_jacket_collision_spine_02 [{
   shapeType=Capsule
   length=21.25
   radius=15.25
   shapeName=dyn_cloth_jacket_collision_spine_02_capsule
   center=<3.6308 -31.0065 0>
   rotation=<rot 0 -79.708 0>
  }]
 foot_l []
 foot_r []
 pelvis [
  {shapeType=Capsule, length=10.75, radius=13.125, shapeName=pelvis_capsule, center=<0.1602 -0.6905 0>}
  {shapeType=Capsule, length=10.75, radius=13.625, shapeName=pelvis_capsule, center=<-6.3893 -0.7666 0>}
 ]
 spine_01 [{shapeType=Capsule, length=14.3438, radius=11.8086, shapeName=spine_01_capsule, center=<-1.7558 -1.1124 0>}]
 spine_02 [
  {shapeType=Capsule, length=7.8125, radius=11.8822, center=<-0.5096 -1.1637 -5.9861>, rotation=<rot -89.7173 3.8121 0>}
  {
   shapeType=Capsule
   length=3.3125
   radius=11.5385
   shapeName=spine_02_capsule
   center=<0.5206 -1.0187 5.9047>
   rotation=<rot -88.9869 -176.1835 -180>
  }
 ]
 spine_03 [
  {
   shapeType=Capsule
   length=6.0625
   radius=11.5625
   shapeName=spine_03_capsule
   center=<1.4703 0.03 -5.9318>
   rotation=<rot 0 -90.0001 87.0476>
  }
  {shapeType=Capsule, length=6.0625, radius=11.5625, shapeName=spine_03_capsule, center=<1.4703 0.03 5.9318>, rotation=<rot 0 -90.0006 87.0476>}
 ]
 thigh_l []
 thigh_r []
}
constraints [
 {
  bone1Name=spine_01
  bone2Name=pelvis
  swing1LimitDegrees=45
  swing1Motion=Free
  swing2LimitDegrees=45
  swing2Motion=Free
  twistLimitDegrees=45
  twistMotion=Free
 }
 {
  bone1Name=spine_02
  bone2Name=spine_01
  swing1LimitDegrees=45
  swing1Motion=Free
  swing2LimitDegrees=45
  swing2Motion=Free
  twistLimitDegrees=45
  twistMotion=Free
 }
 {
  bone1Name=thigh_l
  bone2Name=pelvis
  swing1LimitDegrees=45
  swing1Motion=Free
  swing2LimitDegrees=45
  swing2Motion=Free
  twistLimitDegrees=45
  twistMotion=Free
 }
 {
  bone1Name=foot_l
  bone2Name=calf_l
  swing1LimitDegrees=45
  swing1Motion=Free
  swing2LimitDegrees=45
  swing2Motion=Free
  twistLimitDegrees=45
  twistMotion=Free
 }
 {
  bone1Name=foot_r
  bone2Name=calf_r
  swing1LimitDegrees=45
  swing1Motion=Free
  swing2LimitDegrees=45
  swing2Motion=Free
  twistLimitDegrees=45
  twistMotion=Free
 }
 {
  bone1Name=calf_r
  bone2Name=thigh_r
  swing1LimitDegrees=45
  swing1Motion=Free
  swing2LimitDegrees=45
  swing2Motion=Free
  twistLimitDegrees=45
  twistMotion=Free
 }
 {
  bone1Name=calf_l
  bone2Name=thigh_l
  swing1LimitDegrees=45
  swing1Motion=Free
  swing2LimitDegrees=45
  swing2Motion=Free
  twistLimitDegrees=45
  twistMotion=Free
 }
 {
  bone1Name=dyn_cloth_jacket_collision_spine_01
  bone2Name=pelvis
  swing1LimitDegrees=45
  swing1Motion=Free
  swing2LimitDegrees=45
  swing2Motion=Free
  twistLimitDegrees=45
  twistMotion=Free
 }
 {
  bone1Name=dyn_cloth_jacket_collision_spine_02
  bone2Name=spine_01
  swing1LimitDegrees=45
  swing1Motion=Free
  swing2LimitDegrees=45
  swing2Motion=Free
  twistLimitDegrees=45
  twistMotion=Free
 }
 {
  bone1Name=spine_03
  bone2Name=spine_02
  swing1LimitDegrees=45
  swing1Motion=Free
  swing2LimitDegrees=45
  swing2Motion=Free
  twistLimitDegrees=45
  twistMotion=Free
 }
 {
  bone1Name=thigh_r
  bone2Name=pelvis
  swing1LimitDegrees=45
  swing1Motion=Free
  swing2LimitDegrees=45
  swing2Motion=Free
  twistLimitDegrees=45
  twistMotion=Free
 }
]
