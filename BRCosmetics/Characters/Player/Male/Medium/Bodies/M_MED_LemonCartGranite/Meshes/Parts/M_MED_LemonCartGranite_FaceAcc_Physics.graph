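# Physics asset for the LemonCartGranite male medium character's face accessory, defining two mirrored seven-bone dynamic hair chains, left and right.
# Each chain is a series of capsules constrained from the head down, stiff at the root with swing limits widening toward the tip and twist kept narrow.
# Head, neck and both clavicle capsules supply body volumes; the clavicles have no constraints, and the head-to-neck joint rotates freely.
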PhysicsAsset /BRCosmetics/Characters/Player/Male/Medium/Bodies/M_MED_LemonCartGranite/Meshes/Parts/M_MED_LemonCartGranite_FaceAcc_Physics
bodies {
 clavicle_l [{shapeType=Capsule, length=16.125, radius=8.63, shapeName=clavicle_l_capsule, center=<7.3779 0 -1.1928>, rotation=<rot -0.0002 -90 102.6551>}]
 clavicle_r [{shapeType=Capsule, length=16.125, radius=8.63, shapeName=clavicle_l_capsule, center=<-7.3779 0 1.1928>, rotation=<rot -0.0002 -90 -77.3449>}]
 dyn_hair_l_01 [{shapeType=Capsule, length=1.4054, radius=1.3177, shapeName=dyn_hair_l_01_capsule, center=<1 0 0>, rotation=<rot 90 0 0>}]
 dyn_hair_l_02 [{
   shapeType=Capsule
   length=2.0835
   radius=1.8489
   shapeName=dyn_hair_l_02_capsule
   center=<1.5492 0 0>
   rotation=<rot 81.1388 -179.9997 -179.9998>
  }]
 dyn_hair_l_03 [{shapeType=Capsule, length=3.2465, radius=1.6581, shapeName=dyn_hair_l_03_capsule, center=<1 0 0>, rotation=<rot 90 0 0>}]
 dyn_hair_l_04 [{shapeType=Capsule, length=2.6813, radius=1.3805, shapeName=dyn_hair_l_04_capsule, center=<1 0 0>, rotation=<rot 90 0 0>}]
 dyn_hair_l_05 [{shapeType=Capsule, length=1.6871, radius=1.3704, shapeName=dyn_hair_l_05_capsule, center=<1 0 0>, rotation=<rot 90 0 0>}]
 dyn_hair_l_06 [{shapeType=Capsule, length=1.3621, radius=0.7522, shapeName=dyn_hair_l_06_capsule, center=<1 0 0>, rotation=<rot 90 0 0>}]
 dyn_hair_l_07 [{shapeType=Capsule, length=1.0105, radius=0.8709, shapeName=dyn_hair_l_07_capsule, center=<1 0 0>, rotation=<rot 90 0 0>}]
 dyn_hair_r_01 [{shapeType=Capsule, length=1.4054, radius=1.3177, shapeName=dyn_hair_l_01_capsule, center=<-1 0 0>, rotation=<rot 90 180 0>}]
 dyn_hair_r_02 [{
   shapeType=Capsule
   length=2.0835
   radius=1.8489
   shapeName=dyn_hair_l_02_capsule
   center=<-1.5492 0 0>
   rotation=<rot 81.1388 -179.9997 0.0002>
  }]
 dyn_hair_r_03 [{shapeType=Capsule, length=3.2465, radius=1.6581, shapeName=dyn_hair_l_03_capsule, center=<-1 0 0>, rotation=<rot 90 180 0>}]
 dyn_hair_r_04 [{shapeType=Capsule, length=2.6813, radius=1.3805, shapeName=dyn_hair_l_04_capsule, center=<-1 0 0>, rotation=<rot 90 180 0>}]
 dyn_hair_r_05 [{shapeType=Capsule, length=1.6871, radius=1.3704, shapeName=dyn_hair_l_05_capsule, center=<-1 0 0>, rotation=<rot 90 180 0>}]
 dyn_hair_r_06 [{shapeType=Capsule, length=1.3621, radius=0.7522, shapeName=dyn_hair_l_06_capsule, center=<-1 0 0>, rotation=<rot 90 180 0>}]
 dyn_hair_r_07 [{shapeType=Capsule, length=1.0105, radius=0.8709, shapeName=dyn_hair_l_07_capsule, center=<-1 0 0>, rotation=<rot 90 180 0>}]
 head [{
   shapeType=Capsule
   length=2.6404
   radius=7.6437
   shapeName=head_capsule
   center=<6.2728 2.1828 0.0092>
   rotation=<rot -0.2469 -87.1356 -89.9541>
  }]
 neck_01 [{
   shapeType=Capsule
   length=7.9375
   radius=5.8175
   shapeName=neck_01_capsule
   center=<4.0339 -0.0001 0.0043>
   rotation=<rot 0 -90.0019 89.9387>
  }]
}
constraints [
 {bone1Name=dyn_hair_l_01, bone2Name=head, swing1LimitDegrees=4, swing2LimitDegrees=4, twistLimitDegrees=4}
 {bone1Name=dyn_hair_l_02, bone2Name=dyn_hair_l_01, swing1LimitDegrees=4, swing2LimitDegrees=4, twistLimitDegrees=4}
 {bone1Name=dyn_hair_l_03, bone2Name=dyn_hair_l_02, swing1LimitDegrees=24, swing2LimitDegrees=24, twistLimitDegrees=8}
 {bone1Name=dyn_hair_l_04, bone2Name=dyn_hair_l_03, swing1LimitDegrees=24, swing2LimitDegrees=24, twistLimitDegrees=8}
 {bone1Name=dyn_hair_l_05, bone2Name=dyn_hair_l_04, swing1LimitDegrees=38, swing2LimitDegrees=38, twistLimitDegrees=8}
 {bone1Name=dyn_hair_l_06, bone2Name=dyn_hair_l_05, swing1LimitDegrees=38, swing2LimitDegrees=38, twistLimitDegrees=8}
 {bone1Name=dyn_hair_l_07, bone2Name=dyn_hair_l_06, swing1LimitDegrees=38, swing2LimitDegrees=38, twistLimitDegrees=8}
 {bone1Name=dyn_hair_r_01, bone2Name=head, swing1LimitDegrees=4, swing2LimitDegrees=4, twistLimitDegrees=4}
 {bone1Name=dyn_hair_r_02, bone2Name=dyn_hair_r_01, swing1LimitDegrees=4, swing2LimitDegrees=4, twistLimitDegrees=4}
 {bone1Name=dyn_hair_r_03, bone2Name=dyn_hair_r_02, swing1LimitDegrees=24, swing2LimitDegrees=24, twistLimitDegrees=8}
 {bone1Name=dyn_hair_r_04, bone2Name=dyn_hair_r_03, swing1LimitDegrees=24, swing2LimitDegrees=24, twistLimitDegrees=8}
 {bone1Name=dyn_hair_r_05, bone2Name=dyn_hair_r_04, swing1LimitDegrees=38, swing2LimitDegrees=38, twistLimitDegrees=8}
 {bone1Name=dyn_hair_r_06, bone2Name=dyn_hair_r_05, swing1LimitDegrees=38, swing2LimitDegrees=38, twistLimitDegrees=8}
 {bone1Name=dyn_hair_r_07, bone2Name=dyn_hair_r_06, swing1LimitDegrees=38, swing2LimitDegrees=38, twistLimitDegrees=8}
 {
  bone1Name=head
  bone2Name=neck_01
  swing1LimitDegrees=45
  swing1Motion=Free
  swing2LimitDegrees=45
  swing2Motion=Free
  twistLimitDegrees=45
  twistMotion=Free
 }
]
